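# Master material for Transit set dirt decals, using the translucent deferred decal domain with unlit shading.
# A Decal texture parameter, defaulting to a BurntTires graffiti texture, has its alpha multiplied by an Asteria ice-noise mask and dimmed by a constant.
# The Opacity parameter blends from that broken-up alpha to the plain alpha, which it yields at its default.
# Metallic, Specular and Roughness are scalar parameters.
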Material /Game/Environments/Helios/Sets/Transit/Materials/MD_Transit_Dirt_Decal_A_Master
graph {
 properties {
  bAutomaticallySetUsageInEditor=false
  blendMode=BLEND_Translucent
  materialDomain=MD_DeferredDecal
  pixelDepthOffsetMode=PDOM_Legacy
  shadingModel=MSM_Unlit
  usedShadingModels=Unlit
 }
 nodes [
  n1 = TextureSample() {samplerType=SAMPLERTYPE_Masks, texture=/Game/Environments/Asteria/Terrain/Textures/T_Asteria_Ice_Noise_M}
  n2 = TextureSampleParameter2D() {parameterName=Decal, texture=/Game/Environments/Helios/Sets/BurntTires/Textures/Graffiti/T_BurntTires_Graffiti_B}
  LinearInterpolate(A=Multiply(A=Constant() {r=0.4}, B=Multiply(A=@n1.RGB, B=@n2.A)), B=@n2.A, Alpha=ScalarParameter() {defaultValue=1, parameterName=Opacity})
  ScalarParameter() {parameterName=Metallic}
  ScalarParameter() {defaultValue=1, parameterName=Specular}
  ScalarParameter() {defaultValue=0.5, parameterName=Roughness}
 ]
}
