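# Material function that outputs a mask selecting one layer by an index stored in the second UV channel, named for four-layer use in the Helios Transit set.
# It floors the U coordinate of UV1 to get a whole-number layer ID per UV position.
# A radius-1 SphereMask compares that ID with the LayerChoice input, serving as an equality test for LayerMask.
# LayerChoice falls back to a "Sample" scalar parameter when unconnected.
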MaterialFunction /Game/Environments/Helios/Sets/Transit/Materials/MF_4LayerPick
graph {nodes [
  n1 = ComponentMask(None=TextureCoordinate() {coordinateIndex=1}) {r=true}
  n2 = ScalarParameter() {defaultValue=2, parameterName=Sample}
  FunctionOutput(None=SphereMask(A=Floor(None=@n1), B=FunctionInput(Preview=@n2) {bUsePreviewValueAsDefault=true, inputName=LayerChoice, inputType=FunctionInput_Scalar}) {attenuationRadius=1}) {outputName=LayerMask}
 ]}
