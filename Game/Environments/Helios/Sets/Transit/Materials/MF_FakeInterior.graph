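# Material function that adds fake interior-mapped rooms to window surfaces of the Helios Transit set, statically switchable via UseFakeInterior.
# An interior-cubemap function feeds long-lat UVs, noise-distorted by pixel depth, into a 2D texture sampled with explicit derivatives.
# Time-of-day functions drive tint, contrast and brightness; mobile substitutes cheaper bump-offset sine gradients.
# The resulting window attributes blend over the input material, with blinds and layer masks also declared.
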MaterialFunction /Game/Environments/Helios/Sets/Transit/Materials/MF_FakeInterior
graph {nodes [
  n1 = ScalarParameter() {defaultValue=0.8, group=Fake Windows, parameterName=InteriorCubeMapScale_X}
  Add(A=@n1)
  n2 = TextureCoordinate()
  n3 = ScalarParameter() {defaultValue=24, group=Fake Windows, parameterName=NoiseTile}
  n4 = Multiply(A=TextureSample(UVs=Multiply(A=@n2, B=@n3)) {samplerType=SAMPLERTYPE_Masks, texture=/Game/Environments/Helios/Sets/Transit/Textures/Mask_M1}.R) {constB=2}
  n5 = ScalarParameter() {defaultValue=0.05, group=Fake Windows, parameterName=DistortCubemapIntensity_Near}
  n6 = ScalarParameter() {defaultValue=0.05, group=Fake Windows, parameterName=DistortCubemapIntensity_Far}
  n7 = PixelDepth()
  n8 = ScalarParameter() {defaultValue=10000, group=Fake Windows, parameterName=DistortionNoiseDistance}
  n9 = NamedRerouteDeclaration(None=Saturate(None=Divide(A=@n7.R, B=@n8))) {name=DistanceMask, nodeColor=<rgba 0.9219 0 1 1>}
  n10 = Add(
   A=TextureCoordinate()
   B=AppendVector(
    A=ScalarParameter() {group=Fake Windows, parameterName=InteriorOffset_X}
    B=ScalarParameter() {defaultValue=0.03, group=Fake Windows, parameterName=InteriorOffset_Y}
   )
  )
  n11 = Multiply(A=ComponentMask(None=NamedRerouteUsage()) {g=true, r=true}, B=ScalarParameter() {group=Fake Windows, parameterName=WindowRefractionScale})
  n12 = NamedRerouteDeclaration(None=Constant() {r=1}) {name=WindowRandomMask.R, nodeColor=<rgba 0.9922 1 0 1>}
  n13 = Constant() {r=1}
  n14 = AppendVector(A=@n12, B=NamedRerouteDeclaration(None=@n13) {name=WindowRandomMask.G, nodeColor=<rgba 0 1 0 1>})
  n15 = ScalarParameter() {defaultValue=2.33, group=Fake Windows, parameterName=RandomMult}
  n16 = MaterialFunctionCall!(
   UVs=NamedRerouteDeclaration(None=Add(A=@n10, B=@n11)) {name=InteriorCubeMapUV, nodeColor=<rgba 1 0 0.2344 1>}
   Custom Random Input=Ceil(None=Multiply(A=@n14, B=@n15))
   Tiling=AppendVector(A=@n1, B=ScalarParameter() {defaultValue=0.8, group=Fake Windows, parameterName=InteriorCubeMapScale_Y})
  ) {materialFunction=/Game/Environments/Asteria/Sets/NeonCity/Techart/MF_NeonCity_InteriorCubemap}
  n17 = QualitySwitch(
   Default=@n16.UVW
   Epic=StaticSwitch(
    True=@n16.UVW Randomized
    False=@n16.UVW Randomized
    Value=StaticBoolParameter() {group=Fake Windows, parameterName=RandomizeCubemapRotation}
   )
  )
  n18 = Add(
   A=Multiply(A=Subtract(A=@n4), B=LinearInterpolate(A=@n5, B=@n6, Alpha=@n9))
   B=MaterialFunctionCall!(Vector=NamedRerouteDeclaration(None=@n17) {name=CubemapUVW, nodeColor=<rgba 1 0 0.5156 1>}) {materialFunction=/Engine/Functions/Engine_MaterialFunctions01/Coordinates/LongLatToUV}.Result
  )
  n19 = ScalarParameter() {defaultValue=1, group=Fake Windows, parameterName=TextureArrayLength}
  n20 = NamedRerouteUsage()
  AppendVector(A=@n18, B=Round(None=Multiply(A=@n19, B=@n20)))
  n21 = TextureCoordinate()
  n22 = ScalarParameter() {defaultValue=1, group=Fake Windows, parameterName=DebugUV2}
  AppendVector(A=Frac(None=Multiply(A=@n21, B=@n22)), B=Constant())
  n23 = FunctionInput(Preview=Constant()) {bUsePreviewValueAsDefault=true, inputName=Mask_Input, sortPriority=1}
  ChannelMaskParameter(None=@n23) {group=Fake Windows, maskChannel=Green, parameterName=FakeInteriorMask, defaultValue=<rgba 0 1 0 0>}
  ComponentMask(None=@n23) {g=true}
  Constant3Vector() {constant=<rgba 0.6179 0.9527 2.3 1>}
  Constant3Vector() {constant=<rgba 0.0816 0.0708 0.0656 1>}
  Constant3Vector() {constant=<rgba 1.1379 1.6437 1.8 1>}
  Constant3Vector() {constant=<rgba 1.4 0.9836 0.9028 1>}
  Constant3Vector() {constant=<rgba 1.2535 0.9286 0.8857 1>}
  Constant3Vector() {constant=<rgba 0.4263 0.5191 0.5144 1>}
  Constant3Vector() {constant=<rgba 0.5 0.55 0 1>}
  Constant3Vector() {constant=<rgba 1.3 0.4 0 1>}
  Frac(None=Add(A=@n12, B=NamedRerouteDeclaration(None=@n13) {name=WindowRandomMask.B, nodeColor=<rgba 0 0 1 1>}))
  n24 = NamedRerouteUsage()
  n25 = NamedRerouteUsage()
  n26 = NamedRerouteUsage()
  FunctionOutput(None=StaticSwitchParameter(True=BlendMaterialAttributes(A=@n24, B=@n25, Alpha=@n26) {vertexAttributeBlendType=UseA}, False=NamedRerouteUsage()) {group=Fake Windows, parameterName=UseFakeInterior, sortPriority=-1})
  n27 = GetMaterialAttributes(None=FunctionInput(Preview=SetMaterialAttributes!(MaterialAttributes)) {bUsePreviewValueAsDefault=true, inputName=Material, inputType=FunctionInput_MaterialAttributes}) {attributeGetTypes [#1]}
  NamedRerouteDeclaration(None=@n27.MaterialAttributes) {name=Input_Material, nodeColor=<rgba 0.5469 0 1 1>}
  NamedRerouteDeclaration(None=MaterialFunctionCall!(LayerChoice=ScalarParameter() {defaultValue=1, group=Fake Windows, parameterName=FakeInterior_Layer}) {materialFunction=&/MF_4LayerPick}.LayerMask) {name=SFX_Mask, nodeColor=<rgba 0 1 0.2734 1>}
  n28 = TextureCoordinate()
  n29 = AppendVector(
   A=ScalarParameter() {defaultValue=2, group=Fake Windows, parameterName=BlindsScaleU}
   B=ScalarParameter() {defaultValue=1, group=Fake Windows, parameterName=BlindsScaleV}
  )
  n30 = Constant()
  n31 = NamedRerouteUsage()
  n32 = TextureSample(UVs=Add(A=Multiply(A=@n28, B=@n29), B=AppendVector(A=@n30, B=@n31))) {samplerType=SAMPLERTYPE_Masks, texture=/Game/Environments/Asteria/Sets/NeonCity/Techart/Textures/T_BlindsMask}
  n33 = SphereMask(
   A=@n32.R
   B=Constant()
   Radius=NamedRerouteUsage()
   Hardness=ScalarParameter() {defaultValue=0.5, group=Fake Windows, parameterName=BlindsEdgeHardness}
  )
  NamedRerouteDeclaration(None=OneMinus(None=Subtract(A=@n33, B=@n32.G))) {name=BlindsMask, nodeColor=<rgba 0.0312 1 0 1>}
  n34 = SetMaterialAttributes!(
   MaterialAttributes=NamedRerouteUsage()
   BaseColor=NamedRerouteUsage()
   EmissiveColor=Multiply(A=NamedRerouteUsage(), B=EyeAdaptationInverse().EyeAdaptationInverse)
   Metallic=ScalarParameter() {group=Fake Windows, parameterName=Window_Metallic}
   Specular=ScalarParameter() {defaultValue=0.5, group=Fake Windows, parameterName=WindowSpecular}
   Roughness=ScalarParameter() {defaultValue=0.2, group=Fake Windows, parameterName=Window_Roughness}
  ) {attributeSetTypes [
    #2
    #3
    #4
    #5
    #6
   ]}
  NamedRerouteDeclaration(None=@n34) {name=FakeInterior_Material, nodeColor=<rgba 0 1 0.7656 1>}
  NamedRerouteDeclaration(None=@n27.Normal) {name=Input_Normal, nodeColor=<rgba 1 0 0.3984 1>}
  n35 = MaterialFunctionCall(
   Color_Morning=VectorParameter() {group=Fake Windows, parameterName=Mobile_CubemapMorning, defaultValue=<rgba 1.8 0.8801 0.683 1>}.RGB
   Color_Day=VectorParameter() {group=Fake Windows, parameterName=Mobile_CubemapDay, defaultValue=<rgba 1.35 1.3381 1.2502 1>}.RGB
   Color_Evening=VectorParameter() {group=Fake Windows, parameterName=Mobile_CubemapEvening, defaultValue=<rgba 1.5 0.834 0.6703 1>}.RGB
   Color_Night=VectorParameter() {group=Fake Windows, parameterName=Mobile_CubemapNight, defaultValue=<rgba 0.6828 0.9558 1.5 1>}.RGB
  ) {materialFunction=&Color_Time_of_Day}
  n36 = TextureCoordinate()
  n37 = ComponentMask(None=VectorParameter() {group=Fake Windows, parameterName=Mobile_FadeLocation, defaultValue=<rgba 0.5 0.55 0 1>}.RGB) {g=true, r=true}
  n38 = ComponentMask(None=VectorParameter() {group=Fake Windows, parameterName=Mobile_WindowBounds, defaultValue=<rgba 0.9 0.3 0 1>}.RGB) {g=true, r=true}
  n39 = ScalarParameter() {defaultValue=0.1, desc=0.075, group=Fake Windows, parameterName=Mobile_WindowEdgeFalloff}
  n40 = BumpOffset(
   Coordinate=@n36
   Height=ScalarParameter() {defaultValue=-0.35, group=Fake Windows, parameterName=Mobile_Bump Offset}
   HeightRatioInput=LinearInterpolate(Alpha=MaterialFunctionCall!(A=@n36, B=@n37, Bounds=@n38, Edge Falloff=@n39) {materialFunction=/Engine/Functions/Engine_MaterialFunctions01/Gradient/BoxMask-2D}.Result) {constB=0.5}
  ) {heightRatio=0.5}
  n41 = Add(A=Multiply(A=Subtract(A=@n40, B=@n37), B=Constant2Vector() {g=3, r=0.95})) {constB=0.25}
  n42 = MaterialFunctionCall(Float2=ConstantBiasScale(None=Sine(None=@n41))) {materialFunction=&BreakOutFloat2Components}
  n43 = LinearInterpolate(
   A=VectorParameter() {group=Fake Windows, parameterName=Mobile_WindowEmissive_1, defaultValue=<rgba 0.023 0.0178 0.0153 1>}.RGB
   B=VectorParameter() {group=Fake Windows, parameterName=Mobile_WindowEmissive_2, defaultValue=<rgba 0.1214 0.1823 0.1457 1>}.RGB
   Alpha=Multiply(A=@n42.R, B=@n42.R)
  )
  n44 = TextureCoordinate()
  n45 = ComponentMask(None=VectorParameter() {group=Fake Windows, parameterName=FadeLocation, defaultValue=<rgba 0.5 0.55 0 1>}.RGB) {g=true, r=true}
  n46 = ComponentMask(None=VectorParameter() {group=Fake Windows, parameterName=WindowBounds, defaultValue=<rgba 1.3 0.4 0 1>}.RGB) {g=true, r=true}
  n47 = ScalarParameter() {defaultValue=0.1, desc=0.075, group=Fake Windows, parameterName=Mobile_WindowEdgeFalloff}
  n48 = BumpOffset(
   Coordinate=@n44
   Height=Constant()
   HeightRatioInput=LinearInterpolate(Alpha=MaterialFunctionCall(A=@n44, B=@n45, Bounds=@n46, Edge Falloff=@n47) {materialFunction=&BoxMask-2D}.Result) {constB=0.5}
  ) {heightRatio=0.5}
  n49 = Add(A=Multiply(A=Subtract(A=@n48, B=@n45), B=Constant2Vector() {g=3, r=0.95})) {constB=0.25}
  n50 = MaterialFunctionCall(Float2=ConstantBiasScale(None=Sine(None=@n49))) {materialFunction=&BreakOutFloat2Components}
  n51 = Multiply(A=@n50.R, B=@n50.R)
  n52 = ScalarParameter() {defaultValue=1, group=Fake Windows, parameterName=Mobile_Window Lit_Low}
  n53 = ScalarParameter() {defaultValue=1, group=Fake Windows, parameterName=Mobile_Window Lit_High}
  n54 = NamedRerouteUsage()
  n55 = Multiply(
   A=VectorParameter() {group=Fake Windows, parameterName=Mobile_WindowEmissive, defaultValue=<rgba 0.4167 0.3334 0.2372 1>}.RGB
   B=Saturate(None=LinearInterpolate(A=@n52, B=@n53, Alpha=@n54))
  )
  n56 = MaterialFunctionCall(
   Color_Morning=VectorParameter() {group=Fake Windows, parameterName=CubemapMorning, defaultValue=<rgba 2 0.4561 0.5357 1>}.RGB
   Color_Day=VectorParameter() {group=Fake Windows, parameterName=CubemapTintDay, defaultValue=<rgba 1 0.8667 0.6646 1>}.RGB
   Color_Evening=VectorParameter() {group=Fake Windows, parameterName=CubemapEvening, defaultValue=<rgba 2 0.5474 0.0849 1>}.RGB
   Color_Night=VectorParameter() {group=Fake Windows, parameterName=CubemapTintEvening, defaultValue=<rgba 0.3229 0.7422 2 1>}.RGB
  ) {materialFunction=&Color_Time_of_Day}
  n57 = TextureCoordinate()
  n58 = TextureSampleParameter2D!(UVs=@n18, "DDX(UVs)"=DDX(Value=@n57), "DDY(UVs)"=DDY(Value=@n57), Apply View MipBias) {
   group=Fake Windows
   mipValueMode=TMVM_Derivative
   parameterName=T_FakeInteriorCube
   texture=&/T_FakeInterior_Helios_copy
  }
  n59 = Power(
   Base=@n58.RGB
   Exp=MaterialFunctionCall(
    Value_Morning=ScalarParameter() {defaultValue=0.8, group=Fake Windows, parameterName=CubemapContrast_Morning}
    Value_Day=ScalarParameter() {defaultValue=0.8, group=Fake Windows, parameterName=CubemapContrast_Day}
    Value_Evening=ScalarParameter() {defaultValue=0.9, group=Fake Windows, parameterName=CubemapContrast_Evening}
    Value_Night=ScalarParameter() {defaultValue=0.75, group=Fake Windows, parameterName=CubemapContrast_Night}
   ) {materialFunction=&Scalar_Time_of_Day}.Result
  )
  n60 = MaterialFunctionCall(
   Value_Morning=ScalarParameter() {defaultValue=0.2, group=Fake Windows, parameterName=RoomBrightnessMorning}
   Value_Day=ScalarParameter() {defaultValue=0.5, group=Fake Windows, parameterName=RoomBrightnessDay}
   Value_Evening=ScalarParameter() {defaultValue=0.5, group=Fake Windows, parameterName=RoomBrightnessEvening}
   Value_Night=ScalarParameter() {defaultValue=0.5, group=Fake Windows, parameterName=RoomBrightnessNight}
  ) {materialFunction=&Scalar_Time_of_Day}
  n61 = Multiply(A=@n60.Result, B=ScalarParameter() {defaultValue=0.4, group=Fake Windows, parameterName=FakeInteriorDistanceFadeAmt}) {constB=0.2}
  n62 = NamedRerouteUsage()
  n63 = ShadingPathSwitch(
   Default=Multiply(A=Multiply(A=@n56.Result, B=@n59), B=LinearInterpolate(A=@n60.Result, B=@n61, Alpha=@n62))
   Mobile=StaticSwitchParameter(True=Multiply(A=@n35.Result, B=@n43), False=Multiply(A=@n51, B=@n55)) {desc=Temp Fix until we have a definitive solution in 34.10, group=Fake Windows, parameterName=Mobile_CH6Mobile_FakeInterior}
  ) {desc=Migrated from FeatureLevelSwitch}
  NamedRerouteDeclaration(None=@n63) {name=WindowDiffuse, nodeColor=<rgba 1 0.5625 0 1>}
  NamedRerouteDeclaration(None=@n63) {name=WindowEmissive, nodeColor=<rgba 0.2422 1 0 1>}
  NamedRerouteUsage()
  n64 = TextureCoordinate()
  n65 = AppendVector(A=NamedRerouteUsage(), B=NamedRerouteUsage())
  n66 = MaterialFunctionCall(Vector=AppendVector(A=Subtract(A=@n65) {constB=0.5}, B=Constant()), Default=Constant3Vector() {constant=<rgba 1 1 0 0>}) {materialFunction=&SafeNormalize}
  n67 = Add(
   A=Constant2Vector() {g=0.5, r=0.5}
   B=Multiply(
    A=ComponentMask(None=@n66.Result) {g=true, r=true}
    B=ScalarParameter() {defaultValue=2, group=Fake Windows, parameterName=Mobile_Window GradientRandom OffsetVector}
   )
  )
  Power(
   Base=Divide(A=Distance(A=@n64, B=@n67), B=ScalarParameter() {defaultValue=3.01, group=Fake Windows, parameterName=Mobile_Window Gradient_Radius})
   Exp=ScalarParameter() {defaultValue=7.36, group=Fake Windows, parameterName=Mobile_Window Gradient_Falloff}
  )
  Sine(None=Divide(A=TextureCoordinate(), B=ScalarParameter() {defaultValue=0.005, group=Fake Windows, parameterName=DistortionNoiseSize}))
 ]}
