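# Material function that adds fake train interiors to window surfaces, optionally blending them over the input material, likely by a layer-pick mask.
# The default path projects an interior cubemap into a lat-long texture with noise refraction and time-of-day grading.
# Mobile instead parallax-shifts a packed 2D texture, adding gradient room colours, skylight reflection, light-direction bloom and view-angle shading.
# Several procedural window-shape experiments and a temporary mobile switch remain unconnected.
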MaterialFunction /Game/Environments/Helios/Sets/Transit/Materials/MF_FakeInterior_Train
graph {nodes [
  n1 = NamedRerouteUsage()
  n2 = NamedRerouteUsage()
  n3 = NamedRerouteUsage()
  FunctionOutput(None=StaticSwitchParameter(True=BlendMaterialAttributes(A=@n1, B=@n2, Alpha=@n3) {vertexAttributeBlendType=UseA}, False=NamedRerouteUsage()) {group=Fake Windows, parameterName=UseFakeInterior, sortPriority=-1})
  Multiply(B=MaterialFunctionCall!(LayerChoice=Constant() {r=1}) {materialFunction=&/MF_4LayerPick}.LayerMask)
  n4 = Transform(None=MaterialFunctionCall() {materialFunction=&CameraDirectionVector}.Result) {transformSourceType=TRANSFORMSOURCE_World, transformType=TRANSFORM_Tangent}
  n5 = Add(
   A=TextureCoordinate()
   B=Multiply(
    A=ComponentMask(None=@n4) {g=true, r=true}
    B=ScalarParameter() {defaultValue=0.7, group=Fake Windows, parameterName=Mobile_BumpOffset, sortPriority=99}
   )
  )
  n6 = AppendVector(
   A=ScalarParameter() {defaultValue=1, group=Fake Windows, parameterName=Mobile_UV_Scale_X, sortPriority=99}
   B=ScalarParameter() {defaultValue=1, group=Fake Windows, parameterName=Mobile_UV_Scale_Y, sortPriority=99}
  )
  n7 = ScalarParameter() {desc=-0.05, group=Fake Windows, parameterName=Mobile_UV_Offset_X, sortPriority=99}
  n8 = ScalarParameter() {defaultValue=-0.14, group=Fake Windows, parameterName=Mobile_UV_Offset_Y, sortPriority=99}
  n9 = Add(A=Add(A=Multiply(A=@n5, B=@n6), B=AppendVector(A=@n7, B=@n8)), B=NamedRerouteUsage())
  n10 = MaterialFunctionCall(Float2=@n9) {materialFunction=&BreakOutFloat2Components}
  n11 = Add(A=@n10.R, B=ScalarParameter() {defaultValue=-0.05, parameterName=test_Window_OffsetX}) {constB=0}
  n12 = ScalarParameter() {defaultValue=2.4, parameterName=test_Window_X}
  n13 = Add(A=Sine(None=Multiply(A=@n11, B=@n12)), B=ScalarParameter() {defaultValue=0.43, parameterName=test_Window_X_Size})
  n14 = Multiply(
   A=Add(A=@n10.G, B=ScalarParameter() {defaultValue=-0.38, parameterName=test_Window_OffsetY}) {constB=0}
   B=ScalarParameter() {defaultValue=7.2, parameterName=test_Window_Y}
  )
  n15 = ScalarParameter() {defaultValue=0.5, parameterName=Param_9}
  n16 = Add(A=OneMinus(None=Distance(A=@n14, B=@n15)), B=ScalarParameter() {defaultValue=0.6, parameterName=test_Window_Y_Size})
  Multiply(A=Saturate(None=@n13), B=Saturate(None=@n16))
  NamedRerouteDeclaration(None=MaterialFunctionCall(LayerChoice=ScalarParameter() {defaultValue=1, group=Fake Windows, parameterName=FakeInterior_Layer}) {materialFunction=&MF_4LayerPick}.LayerMask) {name=SFX_Mask, nodeColor=<rgba 0 1 0.2734 1>}
  n17 = SetMaterialAttributes!(
   MaterialAttributes=NamedRerouteUsage()
   BaseColor=NamedRerouteUsage()
   EmissiveColor=EyeAdaptationInverse(LightValueInput=NamedRerouteUsage(), AlphaInput=ScalarParameter() {defaultValue=1, parameterName=EyeAdaptationInverse}).EyeAdaptationInverse
   Metallic=ScalarParameter() {group=Fake Windows, parameterName=Window_Metallic}
   Specular=ScalarParameter() {defaultValue=0.5, group=Fake Windows, parameterName=WindowSpecular}
   Roughness=ScalarParameter() {defaultValue=0.2, group=Fake Windows, parameterName=Window_Roughness}
  ) {attributeSetTypes [
    #1
    #2
    #3
    #4
    #5
   ]}
  NamedRerouteDeclaration(None=@n17) {name=FakeInterior_Material, nodeColor=<rgba 0 1 0.7656 1>}
  n18 = GetMaterialAttributes(None=FunctionInput(Preview=SetMaterialAttributes!(MaterialAttributes)) {bUsePreviewValueAsDefault=true, inputName=Material, inputType=FunctionInput_MaterialAttributes}) {attributeGetTypes [#6]}
  NamedRerouteDeclaration(None=@n18.Normal) {name=Input_Normal, nodeColor=<rgba 1 0 0.3984 1>}
  n19 = Multiply(A=CameraVectorWS()) {constB=-1}
  n20 = Multiply(A=VertexNormalWS(), B=Constant() {r=0.5})
  NamedRerouteDeclaration(None=SkyLightEnvMapSample(
    Direction=Add(A=@n19, B=@n20)
    Roughness=ScalarParameter() {group=Fake Windows, parameterName=Mobile_Window_Cubemap_Roughness, sortPriority=100}
   )) {name=CubemapColor, nodeColor=<rgba 1 0 0.4688 1>}
  n21 = TextureSampleParameter2D(UVs=@n9) {
   group=Fake Windows
   parameterName=Mobile_InteriorTexture
   sortPriority=99
   texture=/Game/Environments/Helios/Sets/Transit/Textures/T_HeliosTrainInteriorMobile
  }
  NamedRerouteDeclaration(None=@n21.G) {name=M_Windows, nodeColor=<rgba 0 0.1797 1 1>}
  NamedRerouteDeclaration(None=@n18.MaterialAttributes) {name=Input_Material, nodeColor=<rgba 0.5469 0 1 1>}
  n22 = MaterialFunctionCall(
   Color_Morning=VectorParameter() {group=Fake Windows, parameterName=CubemapMorning, defaultValue=<rgba 2 0.4561 0.5357 1>}.RGB
   Color_Day=VectorParameter() {group=Fake Windows, parameterName=CubemapTintDay, defaultValue=<rgba 1 0.8667 0.6646 1>}.RGB
   Color_Evening=VectorParameter() {group=Fake Windows, parameterName=CubemapEvening, defaultValue=<rgba 2 0.5474 0.0849 1>}.RGB
   Color_Night=VectorParameter() {group=Fake Windows, parameterName=CubemapTintEvening, defaultValue=<rgba 0.3229 0.7422 2 1>}.RGB
  ) {materialFunction=&Color_Time_of_Day}
  n23 = TextureSample(UVs=Multiply(A=TextureCoordinate(), B=ScalarParameter() {defaultValue=24, group=Fake Windows, parameterName=NoiseTile})) {samplerType=SAMPLERTYPE_LinearGrayscale, texture=/Game/Environments/Helios/Sets/Transit/Textures/T_TrainGlassRefraction}
  n24 = Saturate(None=Divide(A=PixelDepth().R, B=ScalarParameter() {defaultValue=10000, group=Fake Windows, parameterName=DistortionNoiseDistance}))
  n25 = Multiply(
   A=Subtract(A=Multiply(A=@n23.R) {constB=2})
   B=LinearInterpolate(
    A=ScalarParameter() {defaultValue=0.05, group=Fake Windows, parameterName=DistortCubemapIntensity_Near}
    B=ScalarParameter() {defaultValue=0.05, group=Fake Windows, parameterName=DistortCubemapIntensity_Far}
    Alpha=NamedRerouteDeclaration(None=@n24) {name=DistanceMask, nodeColor=<rgba 0.9219 0 1 1>}
   )
  )
  n26 = ScalarParameter() {group=Fake Windows, parameterName=InteriorOffset_X}
  n27 = ScalarParameter() {defaultValue=0.03, group=Fake Windows, parameterName=InteriorOffset_Y}
  n28 = NamedRerouteUsage()
  n29 = Add(
   A=Add(A=TextureCoordinate(), B=AppendVector(A=@n26, B=@n27))
   B=Multiply(A=ComponentMask(None=@n28) {g=true, r=true}, B=ScalarParameter() {group=Fake Windows, parameterName=WindowRefractionScale})
  )
  n30 = ScalarParameter() {defaultValue=0.8, group=Fake Windows, parameterName=InteriorCubeMapScale_X}
  n31 = ScalarParameter() {defaultValue=0.8, group=Fake Windows, parameterName=InteriorCubeMapScale_Y}
  n32 = NamedRerouteDeclaration(None=MaterialFunctionCall!(
    UVs=NamedRerouteDeclaration(None=@n29) {name=Windows UV, nodeColor=<rgba 1 0 0.2344 1>}
    Custom Random Input
    Tiling=AppendVector(A=@n30, B=@n31)
   ) {materialFunction=/Game/Environments/Asteria/Sets/NeonCity/Techart/MF_NeonCity_InteriorCubemap}.UVW) {name=CubemapUVW, nodeColor=<rgba 1 0 0.5156 1>}
  n33 = TextureCoordinate()
  n34 = TextureSampleParameter2D!(
   UVs=Add(
    A=NamedRerouteDeclaration(None=@n25) {name=RefractionPattern, nodeColor=<rgba 1 0 0.6094 1>}
    B=MaterialFunctionCall!(Vector=@n32) {materialFunction=/Engine/Functions/Engine_MaterialFunctions01/Coordinates/LongLatToUV}.Result
   )
   "DDX(UVs)"=DDX(Value=@n33)
   "DDY(UVs)"=DDY(Value=@n33)
   Apply View MipBias
  ) {
   group=Fake Windows
   mipValueMode=TMVM_Derivative
   parameterName=T_FakeInteriorCube
   texture=&/T_FakeInterior_Helios_copy
  }
  n35 = Power(
   Base=@n34.RGB
   Exp=MaterialFunctionCall(
    Value_Morning=ScalarParameter() {defaultValue=0.8, group=Fake Windows, parameterName=CubemapContrast_Morning}
    Value_Day=ScalarParameter() {defaultValue=0.8, group=Fake Windows, parameterName=CubemapContrast_Day}
    Value_Evening=ScalarParameter() {defaultValue=0.9, group=Fake Windows, parameterName=CubemapContrast_Evening}
    Value_Night=ScalarParameter() {defaultValue=0.75, group=Fake Windows, parameterName=CubemapContrast_Night}
   ) {materialFunction=&Scalar_Time_of_Day}.Result
  )
  n36 = MaterialFunctionCall(
   Value_Morning=ScalarParameter() {defaultValue=0.2, group=Fake Windows, parameterName=RoomBrightnessMorning}
   Value_Day=ScalarParameter() {defaultValue=0.5, group=Fake Windows, parameterName=RoomBrightnessDay}
   Value_Evening=ScalarParameter() {defaultValue=0.5, group=Fake Windows, parameterName=RoomBrightnessEvening}
   Value_Night=ScalarParameter() {defaultValue=0.5, group=Fake Windows, parameterName=RoomBrightnessNight}
  ) {materialFunction=&Scalar_Time_of_Day}
  n37 = Multiply(A=@n36.Result, B=ScalarParameter() {defaultValue=0.4, group=Fake Windows, parameterName=FakeInteriorDistanceFadeAmt}) {constB=0.2}
  n38 = NamedRerouteUsage()
  n39 = VectorParameter() {group=Fake Windows, parameterName=Mobile_CubemapMorning, sortPriority=110, defaultValue=<rgba 1.5 0.9231 0.7995 1>}
  n40 = VectorParameter() {group=Fake Windows, parameterName=Mobile_CubemapDay, sortPriority=110, defaultValue=<rgba 1 0.9912 0.926 1>}
  n41 = VectorParameter() {group=Fake Windows, parameterName=Mobile_CubemapEvening, sortPriority=110, defaultValue=<rgba 1.4 0.7784 0.6256 1>}
  n42 = VectorParameter() {group=Fake Windows, parameterName=Mobile_CubemapNight, sortPriority=110, defaultValue=<rgba 0.4635 0.6428 1 1>}
  n43 = VectorParameter() {group=Fake Windows, parameterName=Mobile_Color_Ceiling, sortPriority=100, defaultValue=<rgba 0.2691 1 0.8541 0.5>}
  n44 = VectorParameter() {group=Fake Windows, parameterName=Mobile_Color_WallTop, sortPriority=100, defaultValue=<rgba 0.1117 0.1193 0.1198 1>}
  n45 = VectorParameter() {group=Fake Windows, parameterName=Mobile_Color_WallBottom, sortPriority=100, defaultValue=<rgba 0.0642 0.0283 0.0253 1>}
  n46 = VectorParameter() {group=Fake Windows, parameterName=Mobile_Color_Floor, sortPriority=100, defaultValue=<rgba 0.0376 0.0559 0.0642 1>}
  n47 = ScalarParameter() {defaultValue=15, group=Fake Windows, parameterName=Mobile_ColorTransitionContrast, sortPriority=100}
  n48 = LinearInterpolate(A=Subtract(B=@n47) {constA=0, desc=Low negative value}, B=Add(A=@n47) {desc=High Positive}, Alpha=ComponentMask(None=@n9) {g=true}) {desc=Lerps between greater range of value}
  n49 = Saturate(None=Add(A=@n48, B=ScalarParameter() {defaultValue=-4, group=Fake Windows, parameterName=Mobile_Color_Floor_Offset, sortPriority=100}))
  n50 = Add(A=@n48, B=ScalarParameter() {defaultValue=0.2, group=Fake Windows, parameterName=Mobile_Color_Wall_Offset, sortPriority=100})
  n51 = ScalarParameter() {defaultValue=5.7, group=Fake Windows, parameterName=Mobile_Color_Ceiling_Offset, sortPriority=100}
  n52 = LinearInterpolate(
   A=@n43.RGB
   B=LinearInterpolate(A=@n44.RGB, B=LinearInterpolate(A=@n45.RGB, B=@n46.RGB, Alpha=@n49), Alpha=Saturate(None=@n50))
   Alpha=Saturate(None=Add(A=@n48, B=@n51))
  )
  n53 = VectorParameter() {group=Fake Windows, parameterName=Mobile_Window_Tint, sortPriority=100, defaultValue=<rgba 0.0932 0.1185 0.13 1>}
  n54 = LinearInterpolate(
   A=@n53.RGB
   B=NamedRerouteUsage()
   Alpha=ScalarParameter() {defaultValue=0.5, group=Fake Windows, parameterName=Mobile_Window_Cubemap, sortPriority=100}
  )
  n55 = NamedRerouteUsage()
  n56 = Multiply(A=@n21.R, B=ScalarParameter() {defaultValue=0.5, desc=0.5, group=Fake Windows, parameterName=Mobile_Texture_Mult, sortPriority=99})
  n57 = Add(
   A=Multiply(A=LinearInterpolate(A=@n52, B=@n54, Alpha=@n55), B=Saturate(None=@n56))
   B=Multiply(A=@n21.R, B=ScalarParameter() {defaultValue=0.05, desc=0.05, group=Fake Windows, parameterName=Mobile_Texture_Add, sortPriority=99})
  )
  n58 = Multiply(
   A=Multiply(A=NamedRerouteUsage(), B=NamedRerouteUsage())
   B=ScalarParameter() {defaultValue=0.1, group=Fake Windows, parameterName=Mobile_WindowBloomLightDir_Mult, sortPriority=100}
  )
  n59 = CameraVectorWS()
  n60 = VertexNormalWS()
  n61 = Multiply(
   A=Multiply(
    A=MaterialFunctionCall(Color_Morning=@n39.RGB, Color_Day=@n40.RGB, Color_Evening=@n41.RGB, Color_Night=@n42.RGB) {materialFunction=&Color_Time_of_Day}.Result
    B=Add(A=@n57, B=@n58)
   )
   B=NamedRerouteDeclaration(None=DotProduct(A=@n59, B=@n60)) {name=M_Fresnel, nodeColor=<rgba 0 0.1328 1 1>}
  )
  n62 = ShadingPathSwitch(Default=Multiply(A=Multiply(A=@n22.Result, B=@n35), B=LinearInterpolate(A=@n36.Result, B=@n37, Alpha=@n38)), Mobile=@n61) {desc=Migrated from FeatureLevelSwitch}
  NamedRerouteDeclaration(None=@n62) {name=WindowDiffuse, nodeColor=<rgba 1 0.5625 0 1>}
  NamedRerouteDeclaration(None=@n62) {name=WindowEmissive, nodeColor=<rgba 0.2422 1 0 1>}
  NamedRerouteDeclaration(None=@n21.B) {name=M_WindowsGloom, nodeColor=<rgba 0.8984 1 0 1>}
  n63 = SkyAtmosphereLightDirection()
  n64 = MaterialFunctionCall() {materialFunction=&CameraDirectionVector}
  NamedRerouteDeclaration(None=Saturate(None=DotProduct(A=@n63, B=@n64.Result))) {name=LightDirection, nodeColor=<rgba 0.9453 0 1 1>}
  n65 = LinearInterpolate(A=Min(A=@n13, B=@n16), B=Multiply(A=@n13, B=@n16), Alpha=ScalarParameter() {defaultValue=0.13, parameterName=test_window_rounded})
  n66 = ScalarParameter() {defaultValue=-1, parameterName=Param_11}
  Saturate(None=Multiply(A=Add(A=@n65, B=@n66), B=ScalarParameter() {defaultValue=5, parameterName=Param_12}))
  n67 = TextureCoordinate()
  n68 = ComponentMask(None=VectorParameter() {group=Fake Windows, parameterName=FadeLocation, defaultValue=<rgba 0.5 0.55 0 1>}.RGB) {g=true, r=true}
  n69 = ComponentMask(None=VectorParameter() {group=Fake Windows, parameterName=WindowBounds, defaultValue=<rgba 1.3 0.4 0 1>}.RGB) {g=true, r=true}
  n70 = ScalarParameter() {defaultValue=0.1, desc=0.075, group=Fake Windows, parameterName=Mobile_WindowEdgeFalloff}
  n71 = BumpOffset(
   Coordinate=@n67
   Height=Constant()
   HeightRatioInput=LinearInterpolate(Alpha=MaterialFunctionCall!(A=@n67, B=@n68, Bounds=@n69, Edge Falloff=@n70) {materialFunction=/Engine/Functions/Engine_MaterialFunctions01/Gradient/BoxMask-2D}.Result) {constB=0.5}
  ) {heightRatio=0.5}
  n72 = Add(A=Multiply(A=Subtract(A=@n71, B=@n68), B=Constant2Vector() {g=3, r=0.95})) {constB=0.25}
  n73 = MaterialFunctionCall(Float2=ConstantBiasScale(None=Sine(None=@n72))) {materialFunction=&BreakOutFloat2Components}
  StaticSwitchParameter(
   True=@n61
   False=Multiply(
    A=Multiply(A=@n73.R, B=@n73.R)
    B=VectorParameter() {group=Fake Windows, parameterName=Mobile_WindowEmissive, defaultValue=<rgba 0.4167 0.3334 0.2372 1>}.RGB
   )
  ) {
   desc=Temp Fix until we have a definitive solution in 34.10
   group=Fake Windows
   parameterName=Mobile_CH6Mobile_FakeInterior
   sortPriority=100
  }
 ]}
