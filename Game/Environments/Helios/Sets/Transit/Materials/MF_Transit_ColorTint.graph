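# Material function that optionally tints Transit set materials with one of four parameterized colors, gated by the static switch Transit_UseColorTint.
# A clamped, normalized custom primitive data value drives four-input lerps, so each primitive can choose its tint.
# Each tint also has min/max remap ranges for metallic, specular and roughness alongside its color multiplier.
# A selectable vertex color channel masks the blend between original and tinted material attributes.
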MaterialFunction /Game/Environments/Helios/Sets/Transit/Materials/MF_Transit_ColorTint
graph {nodes [
  ComponentMask(None=TextureCoordinate()) {r=true}
  Constant()
  n1 = NamedRerouteUsage()
  n2 = SetMaterialAttributes!(
   MaterialAttributes=NamedRerouteUsage()
   BaseColor=NamedRerouteUsage()
   Metallic=NamedRerouteUsage()
   Specular=NamedRerouteUsage()
   Roughness=NamedRerouteUsage()
  ) {attributeSetTypes [
    #1
    #2
    #3
    #4
   ]}
  n3 = NamedRerouteUsage()
  FunctionOutput(None=StaticSwitchParameter(True=BlendMaterialAttributes(A=@n1, B=@n2, Alpha=@n3) {vertexAttributeBlendType=UseA}, False=NamedRerouteUsage()) {parameterName=Transit_UseColorTint})
  n4 = GetMaterialAttributes(None=FunctionInput() {inputName=Material, inputType=FunctionInput_MaterialAttributes}) {attributeGetTypes [
    #1
    #2
    #3
    #4
   ]}
  NamedRerouteDeclaration(None=@n4.BaseColor) {name=Input_BaseColor, nodeColor=<rgba 1 0.4688 0 1>}
  NamedRerouteDeclaration(None=@n4.Metallic) {name=Input_Metallic, nodeColor=<rgba 0.1484 0 1 1>}
  n5 = ScalarParameter() {group=Transit_ColorTinting, parameterName=TintMetallicMin_01}
  n6 = ScalarParameter() {defaultValue=1, group=Transit_ColorTinting, parameterName=TintMetallicMax_01}
  n7 = NamedRerouteUsage()
  n8 = ScalarParameter() {group=Transit_ColorTinting, parameterName=TintMetallicMin_02}
  n9 = ScalarParameter() {defaultValue=1, group=Transit_ColorTinting, parameterName=TintMetallicMax_02}
  n10 = NamedRerouteUsage()
  n11 = ScalarParameter() {group=Transit_ColorTinting, parameterName=TintMetallicMin_03}
  n12 = ScalarParameter() {defaultValue=1, group=Transit_ColorTinting, parameterName=TintMetallicMax_03}
  n13 = NamedRerouteUsage()
  n14 = ScalarParameter() {group=Transit_ColorTinting, parameterName=TintMetallicMin_04}
  n15 = ScalarParameter() {defaultValue=1, group=Transit_ColorTinting, parameterName=TintMetallicMax_04}
  n16 = NamedRerouteUsage()
  NamedRerouteDeclaration(None=MaterialFunctionCall(
    1=LinearInterpolate(A=@n5, B=@n6, Alpha=@n7)
    2=LinearInterpolate(A=@n8, B=@n9, Alpha=@n10)
    3=LinearInterpolate(A=@n11, B=@n12, Alpha=@n13)
    4=LinearInterpolate(A=@n14, B=@n15, Alpha=@n16)
    A=NamedRerouteUsage()
   ) {materialFunction=&Lerp_Multiple_Float3}.Lerp 4 Inputs) {name=FinalMetallic, nodeColor=<rgba 0.7344 1 0 1>}
  NamedRerouteDeclaration(None=@n4.Specular) {name=Input_Specular, nodeColor=<rgba 0.1719 1 0 1>}
  NamedRerouteDeclaration(None=@n4.Roughness) {name=Input_Roughness, nodeColor=<rgba 1 0 0.4922 1>}
  NamedRerouteDeclaration(None=ChannelMaskParameter(None=VertexColor()) {parameterName=ColorTintChannel}.R) {name=ColorTint_Mask, nodeColor=<rgba 0 0.8125 1 1>}
  NamedRerouteDeclaration(None=@n4.MaterialAttributes) {name=Input_MaterialAttributes, nodeColor=<rgba 1 0.8672 0 1>}
  n17 = VectorParameter() {group=Transit_ColorTinting, parameterName=ColorTint_01, defaultValue=<rgba 0.4375 0 0 1>}
  n18 = VectorParameter() {group=Transit_ColorTinting, parameterName=ColorTint_02, defaultValue=<rgba 0.0184 0.1049 0.6528 1>}
  n19 = VectorParameter() {group=Transit_ColorTinting, parameterName=ColorTint_03, defaultValue=<rgba 0.0901 0.7569 0.0781 1>}
  n20 = VectorParameter() {group=Transit_ColorTinting, parameterName=ColorTint_04, defaultValue=<rgba 0.5833 0.2032 0 1>}
  n21 = NamedRerouteUsage()
  NamedRerouteDeclaration(None=Multiply(
    A=NamedRerouteUsage()
    B=MaterialFunctionCall(1=@n17.RGB, 2=@n18.RGB, 3=@n19.RGB, 4=@n20.RGB, A=@n21) {materialFunction=&Lerp_Multiple_Float3}.Lerp 4 Inputs
   )) {name=FinalColor, nodeColor=<rgba 0.5469 0 1 1>}
  n22 = Constant() {r=3}
  n23 = Clamp(None=ScalarParameter() {bUseCustomPrimitiveData=true, parameterName=Transit_ColorTintChoice, primitiveDataIndex=3}, Max=@n22)
  NamedRerouteDeclaration(None=Saturate(None=Divide(A=@n23, B=@n22))) {name=TintChoice, nodeColor=<rgba 0 1 0.2266 1>}
  n24 = ScalarParameter() {group=Transit_ColorTinting, parameterName=TintRoughnessMin_01}
  n25 = ScalarParameter() {defaultValue=1, group=Transit_ColorTinting, parameterName=TintRoughnessMax_01}
  n26 = NamedRerouteUsage()
  n27 = ScalarParameter() {group=Transit_ColorTinting, parameterName=TintRoughnessMin_02}
  n28 = ScalarParameter() {defaultValue=1, group=Transit_ColorTinting, parameterName=TintRoughnessMax_02}
  n29 = NamedRerouteUsage()
  n30 = ScalarParameter() {group=Transit_ColorTinting, parameterName=TintRoughnessMin_03}
  n31 = ScalarParameter() {defaultValue=1, group=Transit_ColorTinting, parameterName=TintRoughnessMax_03}
  n32 = NamedRerouteUsage()
  n33 = ScalarParameter() {group=Transit_ColorTinting, parameterName=TintRoughnessMin_04}
  n34 = ScalarParameter() {defaultValue=1, group=Transit_ColorTinting, parameterName=TintRoughnessMax_04}
  n35 = NamedRerouteUsage()
  NamedRerouteDeclaration(None=MaterialFunctionCall(
    1=LinearInterpolate(A=@n24, B=@n25, Alpha=@n26)
    2=LinearInterpolate(A=@n27, B=@n28, Alpha=@n29)
    3=LinearInterpolate(A=@n30, B=@n31, Alpha=@n32)
    4=LinearInterpolate(A=@n33, B=@n34, Alpha=@n35)
    A=NamedRerouteUsage()
   ) {materialFunction=&Lerp_Multiple_Float3}.Lerp 4 Inputs) {name=FinalRoughness, nodeColor=<rgba 1 0 0.0937 1>}
  n36 = ScalarParameter() {group=Transit_ColorTinting, parameterName=TintSpecularMin_01}
  n37 = ScalarParameter() {defaultValue=1, group=Transit_ColorTinting, parameterName=TintSpecularMax_01}
  n38 = NamedRerouteUsage()
  n39 = ScalarParameter() {group=Transit_ColorTinting, parameterName=TintSpecularMin_02}
  n40 = ScalarParameter() {defaultValue=1, group=Transit_ColorTinting, parameterName=TintSpecularMax_02}
  n41 = NamedRerouteUsage()
  n42 = ScalarParameter() {group=Transit_ColorTinting, parameterName=TintSpecularMin_03}
  n43 = ScalarParameter() {defaultValue=1, group=Transit_ColorTinting, parameterName=TintSpecularMax_03}
  n44 = NamedRerouteUsage()
  n45 = ScalarParameter() {group=Transit_ColorTinting, parameterName=TintSpecularMin_04}
  n46 = ScalarParameter() {defaultValue=1, group=Transit_ColorTinting, parameterName=TintSpecularMax_04}
  n47 = NamedRerouteUsage()
  NamedRerouteDeclaration(None=MaterialFunctionCall(
    1=LinearInterpolate(A=@n36, B=@n37, Alpha=@n38)
    2=LinearInterpolate(A=@n39, B=@n40, Alpha=@n41)
    3=LinearInterpolate(A=@n42, B=@n43, Alpha=@n44)
    4=LinearInterpolate(A=@n45, B=@n46, Alpha=@n47)
    A=NamedRerouteUsage()
   ) {materialFunction=&Lerp_Multiple_Float3}.Lerp 4 Inputs) {name=FinalSpecular, nodeColor=<rgba 0 0.4141 1 1>}
 ]}
