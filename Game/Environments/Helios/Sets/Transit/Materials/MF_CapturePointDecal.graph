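# Material function for a capture point decal that paints graffiti over the input material's base color, toggled by the UseCapturePoint static switch.
# Graffiti is projected triplanar-style from pre-skinned position and normal, optionally blended with a UV2 texture.
# A sphere mask, scaled by object radius and Graffiti custom primitive data, reveals it with a noise-broken hard edge.
# CaptureProgress, progress bar bounds and debug sine appear unconnected to the output.
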MaterialFunction /Game/Environments/Helios/Sets/Transit/Materials/MF_CapturePointDecal
graph {nodes [
  n1 = GetMaterialAttributes(None=FunctionInput(Preview=SetMaterialAttributes!(MaterialAttributes)) {bUsePreviewValueAsDefault=true, inputName=Material, inputType=FunctionInput_MaterialAttributes}) {attributeGetTypes [#1, #2]}
  n2 = NamedRerouteDeclaration(None=@n1.MaterialAttributes) {name=Input_Material, nodeColor=<rgba 0.1016 0 1 1>}
  n3 = NamedRerouteUsage()
  FunctionOutput(None=StaticSwitchParameter(
    True=SetMaterialAttributes!(MaterialAttributes=@n2, BaseColor=@n3) {attributeSetTypes [#1]}
    False=NamedRerouteUsage()
   ) {parameterName=UseCapturePoint}) {outputName=MaterialOutput}
  MaterialFunctionCall(Speed=ScalarParameter() {defaultValue=0.2, parameterName=DebugSpeed}) {materialFunction=&DebugTimeSine}
  n4 = PreSkinnedPosition()
  NamedRerouteDeclaration(None=ShaderStageSwitch(PixelShader=VertexInterpolator(VS=@n4).PS, VertexShader=@n4)) {name=PreskinnedWorldPos, nodeColor=<rgba 0.2187 1 0 1>}
  NamedRerouteDeclaration(None=VectorParameter() {parameterName=Marking_Thickness, defaultValue=<rgba 45 45 0 1>}.RGB) {name=MarkingThickness, nodeColor=<rgba 0.1016 1 0 1>}
  NamedRerouteDeclaration(None=@n1.BaseColor) {name=InputBaseColor, nodeColor=<rgba 1 0 0.5391 1>}
  NamedRerouteDeclaration(None=@n1.EmissiveColor) {name=InputEmissiveColor, nodeColor=<rgba 1 0.1875 0 1>}
  n5 = NamedRerouteUsage()
  n6 = ComponentMask(None=@n5) {a=true}
  n7 = ObjectRadius()
  n8 = ScalarParameter() {bUseCustomPrimitiveData=true, parameterName=Graffiti, primitiveDataIndex=2}
  n9 = SphereMask(
   A=NamedRerouteUsage()
   B=NamedRerouteUsage()
   Radius=Multiply(A=Multiply(A=@n7) {constB=1.5}, B=Multiply(A=@n8) {constB=2})
   Hardness=ScalarParameter() {group=CaptureGraffiti, parameterName=Hardnes}
  ) {hardnessPercent=0}
  n10 = TextureObject() {samplerType=SAMPLERTYPE_Masks, texture=/Game/Global/Textures/Default/Noise/T_Metal_Noise}
  n11 = ScalarParameter() {defaultValue=1024, group=CaptureGraffiti, parameterName=NoiseScale2}
  n12 = NamedRerouteUsage()
  n13 = NamedRerouteUsage()
  n14 = Multiply(
   A=ComponentMask(None=MaterialFunctionCall(TextureObject=@n10, TextureSize=@n11, WorldPosition=@n12, World Space Normal=@n13) {materialFunction=&WorldAlignedTexture}.XYZ Texture) {r=true}
   B=ScalarParameter() {defaultValue=1, group=CaptureGraffiti, parameterName=NoiseStrength}
  )
  n15 = Saturate(None=Ceil(None=Subtract(A=@n9, B=@n14)))
  NamedRerouteDeclaration(None=LinearInterpolate(A=NamedRerouteUsage(), B=ComponentMask(None=@n5) {b=true, g=true, r=true}, Alpha=Multiply(A=@n6, B=@n15))) {name=OutputColor, nodeColor=<rgba 0.0078 1 0 1>}
  n16 = PreSkinnedNormal()
  NamedRerouteDeclaration(None=ShaderStageSwitch(PixelShader=VertexInterpolator(VS=@n16).PS, VertexShader=@n16)) {name=PreskinnedNormal, nodeColor=<rgba 0 1 0.2266 1>}
  NamedRerouteDeclaration(None=VectorParameter() {parameterName=ProgressBar_Bounds, defaultValue=<rgba 18 160 0 1>}.RGB) {name=ProgressBarBounds, nodeColor=<rgba 0.7578 0 1 1>}
  n17 = NamedRerouteUsage()
  n18 = TextureSampleParameter2D(UVs=TextureCoordinate() {coordinateIndex=2}) {
   group=CaptureGraffiti
   parameterName=Graffiti_UV2Texture
   samplerSource=SSM_Wrap_WorldGroupSettings
   texture=/Game/Environments/Helios/Sets/Transit/Textures/T_Transit_UnderGround_Graffiti_MMO_D
  }
  NamedRerouteDeclaration(None=StaticSwitchParameter(True=@n17, False=LinearInterpolate(A=@n17, B=@n18.RGBA, Alpha=@n18.RGBA)) {parameterName=UseOnlyTilingGraffiti}) {name=Graffiti_Color, nodeColor=<rgba 1 0.6563 0 1>}
  n19 = TextureObjectParameter() {
   group=CaptureGraffiti
   parameterName=TilingGraffiti
   texture=/Game/Environments/Helios/Sets/Transit/Textures/T_Transit_Underground_Graffiti_Tiling_D
  }
  NamedRerouteDeclaration(None=@n19) {name=TilingGraffitiTexture, nodeColor=<rgba 0 0.0391 1 1>}
  n20 = Divide(A=NamedRerouteUsage(), B=VectorParameter() {group=CaptureGraffiti, parameterName=Graffiti_Tiling, defaultValue=<rgba 756 756 756 1>}.RGB)
  n21 = TextureSample(UVs=ComponentMask(None=@n20) {b=true, g=true}, Tex=NamedRerouteUsage()) {samplerSource=SSM_Wrap_WorldGroupSettings, texture=&T_Transit_Underground_Graffiti_Tiling_D}
  n22 = TextureSample(UVs=ComponentMask(None=@n20) {g=true, r=true}, Tex=NamedRerouteUsage()) {samplerSource=SSM_Wrap_WorldGroupSettings, texture=&T_Transit_Underground_Graffiti_Tiling_D}
  n23 = Abs(None=NamedRerouteUsage())
  n24 = ComponentMask(None=@n23) {b=true}
  n25 = TextureSample(UVs=ComponentMask(None=@n20) {b=true, r=true}, Tex=NamedRerouteUsage()) {samplerSource=SSM_Wrap_WorldGroupSettings, texture=&T_Transit_Underground_Graffiti_Tiling_D}
  NamedRerouteDeclaration(None=LinearInterpolate(A=LinearInterpolate(A=@n21.RGBA, B=@n22.RGBA, Alpha=@n24), B=@n25.RGBA, Alpha=ComponentMask(None=@n23) {g=true})) {name=TilingGraffiti, nodeColor=<rgba 0 1 0.5547 1>}
  NamedRerouteDeclaration(None=VectorParameter() {parameterName=CenterPos, defaultValue=<rgba 128 0 0 1>}.RGB) {name=LocalObjectPos, nodeColor=<rgba 0.9922 1 0 1>}
  NamedRerouteDeclaration(None=ScalarParameter() {bUseCustomPrimitiveData=true, parameterName=CaptureProgress, primitiveDataIndex=1}) {name=CaptureStatus, nodeColor=<rgba 1 0 0.375 1>}
  NamedRerouteDeclaration(None=VectorParameter() {parameterName=Box_Bounds, defaultValue=<rgba 380 1850 0 1>}.RGB) {name=BoxBounds, nodeColor=<rgba 1 0 0.8437 1>}
 ]}
