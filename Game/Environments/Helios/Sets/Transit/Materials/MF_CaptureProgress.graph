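# Material function that adds a capture-point progress bar and emissive lighting to a material, toggled by the UseCaptureProgressBar static switch.
# Custom primitive data supplies capture state, which blends five team/status colors through MF_Lerp5Values, and capture progress.
# UV1 sine and frac math cuts the mask texture's red channel into cells and shapes a progress mask, while green drives blinking ambient glow.
# The result modifies BaseColor, Emissive, Specular and Roughness.
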MaterialFunction /Game/Environments/Helios/Sets/Transit/Materials/MF_CaptureProgress
graph {nodes [
  n1 = TextureCoordinate() {coordinateIndex=1}
  ConstantBiasScale(None=Sine(None=Add(A=@n1) {constB=-0.25}))
  n2 = TextureCoordinate() {coordinateIndex=1}
  n3 = ScalarParameter() {defaultValue=20, parameterName=Cells}
  n4 = Multiply(A=Subtract(A=ComponentMask(None=@n2) {r=true}) {constB=-0.025}, B=@n3)
  n5 = Divide(A=Floor(None=@n4), B=@n3)
  ConstantBiasScale(None=Sine(None=Add(A=@n5) {constB=-0.25}))
  n6 = GetMaterialAttributes(None=FunctionInput() {inputType=FunctionInput_MaterialAttributes}) {attributeGetTypes [
    #1
    #2
    #3
    #4
   ]}
  n7 = SetMaterialAttributes!(
   MaterialAttributes=NamedRerouteDeclaration(None=@n6.MaterialAttributes) {name=InputMaterial, nodeColor=<rgba 0.0547 0 1 1>}
   BaseColor=NamedRerouteUsage()
   EmissiveColor=NamedRerouteUsage()
   Specular=LinearInterpolate(A=@n6.Specular, B=ScalarParameter() {defaultValue=1, parameterName=EmissiveSPecx}, Alpha=NamedRerouteUsage())
   Roughness=LinearInterpolate(A=@n6.Roughness, B=ScalarParameter() {defaultValue=0.2, parameterName=EmissiveRoughness}, Alpha=NamedRerouteUsage()) {constA=-0.99}
  ) {attributeSetTypes [
    #1
    #2
    #3
    #4
   ]}
  FunctionOutput(None=StaticSwitchParameter(True=@n7, False=NamedRerouteUsage()) {parameterName=UseCaptureProgressBar})
  n8 = ScalarParameter() {group=CaptureProgress, parameterName=Emissive_Strength}
  n9 = Time()
  n10 = ScalarParameter() {defaultValue=0.25, parameterName=BlinkSpeed}
  n11 = ConstantBiasScale(None=Sine(None=Multiply(A=@n9, B=@n10)))
  n12 = LinearInterpolate(A=@n8, B=ScalarParameter() {defaultValue=1, parameterName=DefaultEmissiveStrength}, Alpha=@n11)
  n13 = ScalarParameter() {defaultValue=1, group=CaptureProgress, parameterName=PulseEmissive_Strength}
  n14 = NamedRerouteUsage()
  LinearInterpolate(A=@n12, B=@n13, Alpha=@n14)
  MaterialFunctionCall() {materialFunction=&DebugTimeSine}
  NamedRerouteDeclaration(None=@n6.BaseColor) {name=Input_BaseColor, nodeColor=<rgba 1 0.5859 0 1>}
  n15 = MaterialFunctionCall!(
   1=VectorParameter() {group=CaptureProgress, parameterName=Emissive_Neutral_Color, defaultValue=<rgba 1 0.3732 0.027 1>}.RGB
   2=VectorParameter() {group=CaptureProgress, parameterName=Emissive_Friendly_Color, defaultValue=<rgba 0 0.4637 1 1>}.RGB
   3=VectorParameter() {group=CaptureProgress, parameterName=Emissive_Contested_Color, defaultValue=<rgba 0 0.4637 1 1>}.RGB
   4=VectorParameter() {group=CaptureProgress, parameterName=Emissive_Enemy_Color, defaultValue=<rgba 0 0 0 1>}.RGB
   5=VectorParameter() {group=CaptureProgress, parameterName=Emissive_Completed_Color, defaultValue=<rgba 0 1 0 1>}.RGB
   A=Divide(A=ScalarParameter() {bUseCustomPrimitiveData=true, defaultValue=3, parameterName=CapturePointState}) {constB=4}
  ) {materialFunction=&/MF_Lerp5Values}
  n16 = TextureSampleParameter2D() {
   group=CaptureProgress
   parameterName=CaptureProgressMask
   samplerSource=SSM_Wrap_WorldGroupSettings
   samplerType=SAMPLERTYPE_Masks
   texture=/Game/Environments/Helios/Sets/Transit/Textures/T_Transit_Car_MMO_Box_E
  }
  n17 = Add(A=Frac(None=@n4)) {constB=-0.25}
  n18 = SphereMask(A=ConstantBiasScale(None=Sine(None=@n17)), B=Constant() {r=1}, Hardness=Constant() {r=0.98}) {attenuationRadius=1}
  n19 = NamedRerouteDeclaration(None=Multiply(A=@n16.R, B=@n18)) {name=CaptureEmissiveMask, nodeColor=<rgba 1 0 0.9375 1>}
  n20 = LinearInterpolate(A=@n8, B=@n13, Alpha=NamedRerouteUsage())
  n21 = Multiply(A=Multiply(A=NamedRerouteUsage(), B=ScalarParameter() {defaultValue=1, parameterName=AmbientEmissiveBrightness}), B=NamedRerouteUsage())
  n22 = Max(
   A=Multiply(A=Multiply(A=@n15.Lerp 4 Inputs, B=@n19), B=LinearInterpolate(A=@n8, B=@n20, Alpha=@n14))
   B=Multiply(A=@n15.Lerp 4 Inputs, B=Multiply(A=@n12, B=@n21))
  )
  NamedRerouteDeclaration(None=Add(A=NamedRerouteUsage(), B=@n22)) {name=Output Emissive, nodeColor=<rgba 0 1 0.4375 1>}
  NamedRerouteDeclaration(None=@n6.EmissiveColor) {name=Input_EmissiveColor, nodeColor=<rgba 0.0781 1 0 1>}
  n23 = Multiply(A=@n18, B=ComponentMask(None=OneMinus(None=@n1)) {r=true})
  NamedRerouteDeclaration(None=SphereMask(A=@n23, B=Constant() {r=1}, Radius=NamedRerouteUsage(), Hardness=Constant() {r=0.95})) {name=Progress, nodeColor=<rgba 0.9688 1 0 1>}
  n24 = LinearInterpolate(A=NamedRerouteUsage(), B=VectorParameter() {parameterName=BG-COlor, defaultValue=<rgba 0 0 0 1>}.RGB, Alpha=NamedRerouteUsage())
  n25 = NamedRerouteUsage()
  NamedRerouteDeclaration(None=LinearInterpolate(A=@n24, B=LinearInterpolate(A=@n24, B=@n22, Alpha=@n25), Alpha=NamedRerouteUsage())) {name=OutPut_BaseColor, nodeColor=<rgba 0 1 0.625 1>}
  NamedRerouteDeclaration(None=Ceil(None=NamedRerouteUsage())) {name=bisBeingCapture, nodeColor=<rgba 0 1 0.6484 1>}
  NamedRerouteDeclaration(None=@n16.G) {name=AmbientEmissive, nodeColor=<rgba 0.8047 1 0 1>}
  NamedRerouteDeclaration(None=Max(A=@n16.R, B=@n16.R)) {name=EmissivesMask, nodeColor=<rgba 0.1953 0 1 1>}
  NamedRerouteDeclaration(None=ScalarParameter() {bUseCustomPrimitiveData=true, parameterName=CaptureProgress, primitiveDataIndex=1}) {name=CaptureStatus, nodeColor=<rgba 1 0 0.375 1>}
  NamedRerouteDeclaration(None=@n11) {name=AmbientEmissiveTimer, nodeColor=<rgba 1 0.5859 0 1>}
  n26 = Sine(None=Multiply(A=Time(), B=ScalarParameter() {defaultValue=0.5, parameterName=PulseSpeed}))
  Reroute(None=SphereMask(A=@n23, B=Constant() {r=1}, Radius=ConstantBiasScale(None=@n26), Hardness=Constant()))
  VectorParameter() {parameterName=Ambient Emissive Color, defaultValue=<rgba 1 0.0156 0.0156 1>}
 ]}
