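# Base Nanite-enabled opaque material for the haunted train in the Helios transit set, built on the shared four-layer system.
# Four texture layers with optional world-space mapping and shared SFX masks merge through MF_Layer_4to1.
# An optional Epic-quality subsurface edge-depth tweak follows, using interior-cubemap UVs and Fresnel, then emissive controls and a haunted fake interior.
# Mobile shading fixes specular and roughness to constants before capture-point decal and capture-progress functions finish the chain.
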Material /Game/Environments/Helios/Sets/Transit/Materials/M_TrainHaunted_Base
graph {
 properties {bUseMaterialAttributes=true, bUsedWithNanite=true, pixelDepthOffsetMode=PDOM_Legacy, usedShadingModels=Default Lit}
 nodes [
  n1 = MaterialFunctionCall!(SFX V3=NamedRerouteUsage(), 1=NamedRerouteUsage(), 2=NamedRerouteUsage(), 3=NamedRerouteUsage(), 4=NamedRerouteUsage()) {materialFunction=/Game/Global/Base4Layers_2022/MaterialFunctions/MF_Layer_4to1}
  Composite!(None=@n1.Result)
  Composite!() {subgraphName=L2 Blending Logic help}
  Composite() {subgraphName=L3 Blending Logic help}
  Composite() {subgraphName=L4 Blending Logic help}
  MaterialExpressionReroute_3 = Reroute(None=@n1.Result)
  n2 = GetMaterialAttributes(None=@MaterialExpressionReroute_3.Input) {attributeGetTypes [#1, #2]}
  n3 = StaticSwitchParameter(
   True=StaticSwitchParameter(True=TextureCoordinate() {coordinateIndex=2}, False=TextureCoordinate() {coordinateIndex=1}) {group=B_SSS_EdgeDepth, parameterName=SSS Depth UV2, sortPriority=-30}
   False=TextureCoordinate()
  ) {desc="Default is UV0 ", group=B_SSS_EdgeDepth, parameterName=SSS Depth UV1, sortPriority=-30}
  n4 = AppendVector(
   A=ScalarParameter() {defaultValue=1, group=B_SSS_EdgeDepth, parameterName=SSSDepth X, sortPriority=5}
   B=ScalarParameter() {defaultValue=1, group=B_SSS_EdgeDepth, parameterName=SSSDepth Y, sortPriority=5}
  )
  n5 = Add(
   A=ComponentMask(None=MaterialFunctionCall!(UVs=@n3, Tiling=@n4, Randomize Rotation) {materialFunction=/Engine/Functions/Engine_MaterialFunctions01/Cubemaps/InteriorCubemap}.UVW) {r=true}
   B=Constant() {r=0.98}
  )
  n6 = ScalarParameter() {defaultValue=3, group=B_SSS_EdgeDepth, parameterName=SSS_Depth Edge Blur, sortPriority=3}
  n7 = LinearInterpolate(A=ScalarParameter() {group=B_SSS_EdgeDepth, parameterName=SSS Depth Min, sortPriority=4}, Alpha=Saturate(None=Divide(A=@n5, B=@n6)))
  n8 = ScalarParameter() {group=B_SSS_EdgeDepth, parameterName=SSS Fresnel Min, sortPriority=1}
  n9 = Saturate(None=Fresnel(
    ExponentIn=ScalarParameter() {defaultValue=3, group=B_SSS_EdgeDepth, parameterName=SSS Fresnel Exponent, sortPriority=2}
    Normal=Transform(None=@n2.Normal)
   ) {baseReflectFraction=0, exponent=3})
  n10 = Saturate(None=Multiply(A=Multiply(A=@n2.Subsurface Color, B=@n7), B=LinearInterpolate(B=@n8, Alpha=@n9) {constA=1}))
  n11 = StaticSwitchParameter(
   True=QualitySwitch(
    Default=@MaterialExpressionReroute_3.Input
    Epic=SetMaterialAttributes!(MaterialAttributes=@n2.MaterialAttributes, Subsurface Color=@n10) {attributeSetTypes [#1]}
   )
   False=@MaterialExpressionReroute_3.Input
  ) {group=B_SSS_EdgeDepth, parameterName=SSS_EdgeDepth}
  MaterialExpressionReroute_9 = Reroute(None=@n11)
  n12 = MaterialFunctionCall!(Material=MaterialFunctionCall!(SFX=NamedRerouteUsage(), Material In=@MaterialExpressionReroute_9.Output, SFX Textures RGB=NamedRerouteUsage()) {materialFunction=/Game/Global/Base4Layers_2022/MaterialFunctions/MF_EmissiveControls_2}.Result) {materialFunction=&/MF_FakeInteriorHaunted}
  Composite!(InputPin=@n12.Result) {subgraphName=Mobile Opt}
  MaterialExpressionReroute_20 = Reroute(None=@n12.Result)
  n13 = GetMaterialAttributes(None=@MaterialExpressionReroute_20.InputPin) {attributeGetTypes [#3, #2]}
  n14 = SetMaterialAttributes!(
   MaterialAttributes=@n13.MaterialAttributes
   Specular=Constant() {r=0.5}
   Roughness=Constant() {r=0.8}
   Base Color=@n13.MaterialAttributes
   Normal=@n13.MaterialAttributes
  ) {attributeSetTypes [
    #4
    #5
    #3
    #2
   ]}
  n15 = ShadingPathSwitch(Default=@MaterialExpressionReroute_20.InputPin, Mobile=@n14) {desc=Migrated from FeatureLevelSwitch}
  MaterialExpressionReroute_21 = Reroute(None=@n15)
  MaterialFunctionCall!(In=MaterialFunctionCall!(Material=@MaterialExpressionReroute_21.Output) {materialFunction=&/MF_CapturePointDecal}.MaterialOutput) {materialFunction=&/MF_CaptureProgress}
  n16 = TextureObjectParameter() {
   desc="SFX TextureGrayscale You can reuse that texture across layers and functions. "
   group=0_SFX Global
   parameterName=SFX_Grayscale
   samplerType=SAMPLERTYPE_LinearGrayscale
   sortPriority=0
   texture=/Game/Global/Base4Layers_2022/Textures/T_SFX_Grayscale
  }
  n17 = TextureObjectParameter() {
   desc="SFX UniversalTextureMask Uses Dot to select channel used for various parts of material.You can reuse that texture across layers and functions. "
   group=0_SFX Global
   parameterName=SFX_Mask
   samplerType=SAMPLERTYPE_Masks
   sortPriority=0
   texture=/Game/Global/Base4Layers_2022/Textures/T_SFX_M
  }
  n18 = StaticSwitchParameter(True=@n16, False=@n17) {desc=Make sure size is less than 512 and lower, group=0_SFX Global, parameterName=UseSFXGrayscale}
  NamedRerouteDeclaration(None=MaterialFunctionCall!(In=@n18) {materialFunction=/Game/Global/Base4Layers_2022/MaterialFunctions/MF_SFXTexture}.SFX) {name=SFX V3, nodeColor=<rgba 0.6406 0 1 1>}
  n19 = StaticBoolParameter() {group=4_layer UV, parameterName=Use UV2 for Layer 4}
  n20 = MaterialFunctionCall!(L3 UV2, L2 UV2, L4 UV2=@n19) {materialFunction=/Game/Global/Base4Layers_2022/MaterialFunctions/MF_UVsSet_New}
  n21 = StaticBoolParameter() {desc=Enable to use World Space Mapping for layer. Only Z projection available in Layer 4, group=4_layer UV, parameterName=Layer 4 WS Mapping}
  n22 = StaticBoolParameter() {
   desc="World Space Z only, you need appropriate mask to make it visible. Use Z or map your layers accordingly "
   group=3_layer UV
   parameterName=Layer 3 WSM_Z
  }
  n23 = StaticBool() {value=true}
  n24 = StaticBool()
  n25 = StaticBoolParameter() {group=4_Layer_Blending, parameterName=Blend Overlay Layer 4}
  n26 = StaticSwitch(
   True=StaticBoolParameter() {desc=Use SFX to mask needed area, group=4_Layer_Blending, parameterName=L4 Simple Swatch, sortPriority=-15}
   False=@n24
   Value=@n25
  )
  MaterialExpressionReroute_16 = Reroute(None=@n26)
  n27 = StaticBoolParameter() {desc="World Space XYZ ( Expensive)", group=3_layer UV, parameterName=Layer 3 WSM_XYZ}
  n28 = StaticBool()
  n29 = TextureObjectParameter() {
   group=4_Layer_Nanite
   parameterName="Normals  Nanite 4"
   samplerType=SAMPLERTYPE_Normal
   sortPriority=0
   texture=/Engine/EngineMaterials/FlatNormal
  }
  n30 = TextureObjectParameter() {
   group=4_Layer_Nanite
   parameterName="SpecularMasks  Nanite 4"
   samplerType=SAMPLERTYPE_Masks
   sortPriority=0
   texture=&T_Fortnite_Default_S
  }
  n31 = TextureObjectParameter() {group=4_Layer_Nanite, parameterName=Diffuse Nanite 4, sortPriority=0, texture=/Game/Global/Textures/Default/Blanks/T_MidgreysRGB}
  n32 = StaticBoolParameter() {desc="Enables SFX for Opacity Mask, mast have Masked blending set in MI", group=4_Layer SFX, parameterName=SFX OpacityMask 4, sortPriority=1}
  n33 = StaticSwitch(True=StaticBoolParameter() {group=4_Layer_Blending, parameterName="L4  Normals Only", sortPriority=-15}, False=@n24, Value=@n25)
  MaterialExpressionReroute_17 = Reroute(None=@n33)
  n34 = TextureObjectParameter() {group=4_Layer, parameterName=Diffuse_Texture_4, sortPriority=0, texture=&T_MidgreysRGB}
  n35 = TextureObjectParameter() {
   group=4_Layer
   parameterName=SpecularMasks_4
   samplerType=SAMPLERTYPE_Masks
   sortPriority=0
   texture=&T_Fortnite_Default_S
  }
  n36 = TextureObjectParameter() {group=4_Layer, parameterName=Normals_Texture_4, samplerType=SAMPLERTYPE_Normal, sortPriority=0, texture=&FlatNormal}
  n37 = VectorParameter() {
   group=4_Layer SFX
   parameterName="SFX_SSS Tint(Expand)_4"
   primitiveDataIndex=2
   sortPriority=0
   channelNames {a=SSS Depth 0-1, b="", g="", r=""}
  }
  n38 = VectorParameter() {
   desc="Use R and G channels of this input to controll ClearCoat amount and Rounghness. It uses SFX texture to mask area affected. "
   group=4_Layer SFX
   parameterName="SFX_ClearCoat(Expand)_4"
   sortPriority=-3
   channelNames {a=Not Used, b=Not Used, g=Clearcoat Roughness, r=Clearcoat Value}
   defaultValue=<rgba 1 0 0 0>
  }
  n39 = VectorParameter() {group=4_Layer SFX, parameterName=SFX_Clearcoat_Ch_4, sortPriority=-4, defaultValue=<rgba 0 0 0 0.5>}
  n40 = ScalarParameter() {desc=If you need to cancel shared Mask texture on specific layer, group=4_Layer SFX, parameterName=SFX_OpacityMaskCancel_4, sortPriority=3}
  n41 = VectorParameter() {
   desc="Emissive color, by default mult by Base color. You can control that blending via Alpha channel of this input. Uses SFX texture as the Mask"
   group=4_Layer SFX
   parameterName="SFX_Emissive_Tint(Expand)_4"
   sortPriority=-6
   channelNames {a=Base color to Emissive, b="", g="", r=""}
   defaultValue=<rgba 0 0 0 1>
  }
  NamedRerouteDeclaration(None=MaterialFunctionCall!(
    WSM Scale=@n20.WSM 4
    Use WS Mapping=@n21
    Use Z=StaticSwitch(True=@n22, False=@n23, Value=@MaterialExpressionReroute_16.Output)
    Use XYZ=StaticSwitch(True=@n27, False=@n28, Value=@MaterialExpressionReroute_16.Output)
    WSM_ProjectionTransitionContrast
    UV2 True=@n19
    WSM_HeightLerpUsingAlpha
    WSM_HeightLerpUsingAlpha_Contrast
    UV's=@n20.WSM 4
    Normals_NN=@n29
    SpecularMasks_NN=@n30
    Diffuse_NN=@n31
    Use SFX Emissive=StaticBoolParameter() {desc=Enables SFX, group=4_Layer SFX, parameterName=SFX Emissive 4, sortPriority=-8}
    SFX Opacity Mask=@n32
    SFX SSS Two Sided Foliange=StaticBoolParameter() {group=4_Layer SFX, parameterName=SFX_SSS Foliage 4, sortPriority=-2}
    SFX SSS=StaticBoolParameter() {group=4_Layer SFX, parameterName=SFX_SSS 4, sortPriority=-2}
    SFX Clearcoat=StaticBoolParameter() {group=4_Layer SFX, parameterName="SFX_Clearcoat  4", sortPriority=-5}
    Diffuse=NamedRerouteDeclaration(None=@n34) {name=D4, nodeColor=<rgba 0 1 0.1797 1>}
    World Position Z=ScalarParameter() {group=4_layer UV, parameterName=L4 WP Offset Z, sortPriority=1}
    SFX Texture Complete=NamedRerouteUsage()
    SFX Tint Channel=Constant3Vector()
    SpecularMasks=NamedRerouteDeclaration(None=@n35) {name=S4, nodeColor=<rgba 1 0.5625 0 1>}
    Normals=NamedRerouteDeclaration(None=@n36) {name=N4, nodeColor=<rgba 0.7812 1 0 1>}
    Normals Power=ScalarParameter() {defaultValue=1, group=4_Layer, parameterName=NormalsValue_4, sortPriority=1}
    "SFX  SSS Channel"=VectorParameter() {group=4_Layer SFX, parameterName=SFX_SSS_Ch_4, sortPriority=-1}.RGB
    SSS Color Value=AppendVector(A=@n37, B=@n37)
    "Clearcoat (R) CC Roughness (G)"=@n38
    Clearcoat Mask Channal=@n39.RGB
    SFX Opacilty Mask Channel=VectorParameter() {group=4_Layer SFX, parameterName=SFX_OpacityMask_Ch_4, sortPriority=2}.RGB
    Opacilty Mask Cancel=@n40
    "SFX Emissive  Channel"=VectorParameter() {group=4_Layer SFX, parameterName=SFX_Emissive_Ch_4, sortPriority=-7}.RGB
    Emissive Tint=AppendVector(A=@n41, B=@n41)
    Data Ramp RGB
    Output Diffuse Only=@MaterialExpressionReroute_16.Output
    Output Normals Only=@MaterialExpressionReroute_17.Output2
   ) {materialFunction=/Game/Global/Base4Layers_2022/MaterialFunctions/MF_CombineLayers}.Result) {name=L4, nodeColor=<rgba 0.6172 0 1 1>}
  n42 = StaticBoolParameter() {group=3_layer UV, parameterName=Use UV2 for Layer 3}
  n43 = MaterialFunctionCall(L3 UV2=@n42) {materialFunction=&MF_UVsSet_New}
  n44 = StaticBoolParameter() {desc=Enable to use World Space Mapping for layer. Only XY or Z available in Layer 3, group=3_layer UV, parameterName=Layer 3 WS Mapping}
  n45 = StaticBoolParameter() {
   desc="World Space Z only, you need appropriate mask to make it visible. Use Z or map your layers accordingly "
   group=3_layer UV
   parameterName=Layer 3 WSM_Z
  }
  n46 = StaticSwitch(True=@n45, False=StaticBool() {value=true}, Value=StaticBoolParameter() {group=4_Layer, parameterName=Use 4 Layers, sortPriority=-1})
  n47 = StaticBool()
  n48 = StaticBoolParameter() {group=3_Layer_Blending, parameterName=Blend Overlay Layer 3}
  n49 = StaticSwitch(
   True=StaticBoolParameter() {desc=Use SFX to mask needed area, group=3_Layer_Blending, parameterName=L3 Simple Swatch, sortPriority=-15}
   False=@n47
   Value=@n48
  )
  MaterialExpressionReroute_14 = Reroute(None=@n49)
  n50 = StaticBoolParameter() {desc="World Space XYZ ( Expensive)", group=3_layer UV, parameterName=Layer 3 WSM_XYZ}
  n51 = StaticBool()
  n52 = TextureObjectParameter() {
   group=3_Layer_Nanite
   parameterName="Normals  Nanite 3"
   samplerType=SAMPLERTYPE_Normal
   sortPriority=0
   texture=&FlatNormal
  }
  n53 = TextureObjectParameter() {
   group=3_Layer_Nanite
   parameterName="SpecularMasks  Nanite 3"
   samplerType=SAMPLERTYPE_Masks
   sortPriority=0
   texture=&T_Fortnite_Default_S
  }
  n54 = TextureObjectParameter() {group=3_Layer_Nanite, parameterName=Diffuse Nanite 3, sortPriority=0, texture=&T_MidgreysRGB}
  n55 = StaticBoolParameter() {desc="Enables SFX for Opacity Mask, mast have Masked blending set in MI", group=3_Layer SFX, parameterName=SFX OpacityMask 3, sortPriority=1}
  n56 = StaticSwitch(True=StaticBoolParameter() {group=3_Layer_Blending, parameterName="L3  Normals Only", sortPriority=-15}, False=@n47, Value=@n48)
  MaterialExpressionReroute_15 = Reroute(None=@n56)
  n57 = TextureObjectParameter() {group=3_Layer, parameterName=Diffuse_Texture_3, sortPriority=0, texture=&T_MidgreysRGB}
  n58 = TextureObjectParameter() {
   group=3_Layer
   parameterName=SpecularMasks_3
   samplerType=SAMPLERTYPE_Masks
   sortPriority=0
   texture=&T_Fortnite_Default_S
  }
  n59 = TextureObjectParameter() {group=3_Layer, parameterName=Normals_Texture_3, samplerType=SAMPLERTYPE_Normal, sortPriority=0, texture=&FlatNormal}
  n60 = VectorParameter() {
   group=3_Layer SFX
   parameterName="SFX_SSS Tint(Expand)_3"
   primitiveDataIndex=2
   sortPriority=0
   channelNames {a=SSS Depth 0-1, b="", g="", r=""}
   defaultValue=<rgba 0 0 0 0.5>
  }
  n61 = VectorParameter() {
   desc="Use R and G channels of this input to controll ClearCoat amount and Rounghness. It uses SFX texture to mask area affected. "
   group=3_Layer SFX
   parameterName="SFX_ClearCoat(Expand)_3"
   sortPriority=-3
   channelNames {a=Not Used, b=Not Used, g=Clearcoat Roughness, r=Clearcoat Value}
   defaultValue=<rgba 1 0 0 0>
  }
  n62 = ScalarParameter() {desc=If you need to cancel shared Mask texture on specific layer, group=3_Layer SFX, parameterName=SFX_OpacityMaskCancel_3, sortPriority=3}
  n63 = VectorParameter() {
   desc="Emissive color, by default mult by Base color. You can control that blending via Alpha channel of this input. Uses SFX texture as the Mask"
   group=3_Layer SFX
   parameterName="SFX_Emissive_Tint(Expand)_3"
   sortPriority=-6
   channelNames {a=Base color to Emissive, b="", g="", r=""}
   defaultValue=<rgba 0 0 0 1>
  }
  NamedRerouteDeclaration(None=MaterialFunctionCall(
    WSM Scale=@n43.WSM 3
    Use WS Mapping=@n44
    Use Z=StaticSwitch(True=@n45, False=@n46, Value=@MaterialExpressionReroute_14.Output2)
    Use XYZ=StaticSwitch(True=@n50, False=@n51, Value=@MaterialExpressionReroute_14.Output2)
    UV2 True=@n42
    UV's=@n43.WSM 3
    Normals_NN=@n52
    SpecularMasks_NN=@n53
    Diffuse_NN=@n54
    Use SFX Emissive=StaticBoolParameter() {desc=Enables SFX, group=3_Layer SFX, parameterName=SFX Emissive 3, sortPriority=-8}
    SFX Opacity Mask=@n55
    SFX SSS Two Sided Foliange=StaticBoolParameter() {group=3_Layer SFX, parameterName=SFX_SSS Foliage 3, sortPriority=-2}
    SFX SSS=StaticBoolParameter() {group=3_Layer SFX, parameterName=SFX_SSS 3, sortPriority=-2}
    SFX Clearcoat=StaticBoolParameter() {group=3_Layer SFX, parameterName="SFX_Clearcoat  3", sortPriority=-5}
    Diffuse=NamedRerouteDeclaration(None=@n57) {name=D3, nodeColor=<rgba 0.9688 1 0 1>}
    World Position Z=ScalarParameter() {group=3_layer UV, parameterName=L3 WP Offset Z, sortPriority=1}
    SFX Texture Complete=NamedRerouteUsage()
    SFX Tint Channel=Constant3Vector()
    SpecularMasks=NamedRerouteDeclaration(None=@n58) {name=S3, nodeColor=<rgba 0.2656 1 0 1>}
    Normals=NamedRerouteDeclaration(None=@n59) {name=N3, nodeColor=<rgba 0.8047 0 1 1>}
    Normals Power=ScalarParameter() {defaultValue=1, group=3_Layer, parameterName=NormalsValue_3, sortPriority=1}
    "SFX  SSS Channel"=VectorParameter() {group=3_Layer SFX, parameterName=SFX_SSS_Ch_3, sortPriority=-1}.RGB
    SSS Color Value=@n60
    "Clearcoat (R) CC Roughness (G)"=@n61
    Clearcoat Mask Channal=VectorParameter() {group=3_Layer SFX, parameterName=SFX_Clearcoat_Ch_3, sortPriority=-4}.RGB
    SFX Opacilty Mask Channel=VectorParameter() {group=3_Layer SFX, parameterName=SFX_OpacityMask_Ch_3, sortPriority=2}.RGB
    Opacilty Mask Cancel=@n62
    "SFX Emissive  Channel"=VectorParameter() {group=3_Layer SFX, parameterName=SFX_Emissive_Ch_3, sortPriority=-7}.RGB
    Emissive Tint=AppendVector(A=@n63, B=@n63)
    Output Diffuse Only=@MaterialExpressionReroute_14.Output2
    Output Normals Only=@MaterialExpressionReroute_15.Output3
   ) {materialFunction=&MF_CombineLayers}.Result) {name=L3, nodeColor=<rgba 0.6406 1 0 1>}
  n64 = TextureObjectParameter() {
   group=1_Layer_Nanite
   parameterName="Normals  Nanite"
   samplerType=SAMPLERTYPE_Normal
   sortPriority=0
   texture=&FlatNormal
  }
  n65 = TextureObjectParameter() {
   group=1_Layer_Nanite
   parameterName="SpecularMasks  Nanite"
   samplerType=SAMPLERTYPE_Masks
   sortPriority=0
   texture=&T_Fortnite_Default_S
  }
  n66 = TextureObjectParameter() {group=1_Layer_Nanite, parameterName=Diffuse Nanite, sortPriority=0, texture=&T_MidgreysRGB}
  n67 = StaticBoolParameter() {desc="Enables SFX for Opacity Mask, mast have Masked blending set in MI", group=1_Layer SFX, parameterName=SFX OpacityMask, sortPriority=1}
  n68 = TextureObjectParameter() {group=1_Layer, parameterName=Diffuse, sortPriority=0, texture=&T_MidgreysRGB}
  n69 = VectorParameter() {desc=Use R or G or B to select your SFX Mask Channel, group=0_SFX BP, parameterName=SFX_Tint_Ch, sortPriority=4, defaultValue=<rgba 1 0 0 0>}
  n70 = TextureObjectParameter() {
   group=1_Layer
   parameterName=SpecularMasks
   samplerType=SAMPLERTYPE_Masks
   sortPriority=0
   texture=&T_Fortnite_Default_S
  }
  n71 = TextureObjectParameter() {group=1_Layer, parameterName=Normals, samplerType=SAMPLERTYPE_Normal, sortPriority=0, texture=&FlatNormal}
  n72 = VectorParameter() {
   group=1_Layer SFX
   parameterName="SFX_SSS Tint(Expand)"
   sortPriority=0
   channelNames {a=SSS Depth 0-1, b="", g="", r=""}
   defaultValue=<rgba 0 0 0 0.5>
  }
  n73 = VectorParameter() {
   desc=Use R and G channels of this input to controll ClearCoat amount and Rounghness. It uses SFX texture to mask area affected.
   group=1_Layer SFX
   parameterName="SFX_ClearCoat(Expand)"
   sortPriority=-3
   channelNames {a=Not Used, b=Not Used, g=Clearcoat Roughness, r=Clearcoat Value}
   defaultValue=<rgba 1 0 0 0>
  }
  n74 = ScalarParameter() {desc=If you need to cancel shared Mask texture on specific layer, group=1_Layer SFX, parameterName=SFX_OpacityMaskCancel, sortPriority=3}
  n75 = VectorParameter() {
   desc="Emissive color, by default mult by Base color. You can control that blending via Alpha channel of this input. Uses SFX texture as the Mask"
   group=1_Layer SFX
   parameterName="SFX_Emissive_Tint(Expand)"
   sortPriority=-6
   channelNames {a=Base color to Emissive, b="", g="", r=""}
   defaultValue=<rgba 0 0 0 1>
  }
  n76 = ScalarParameter() {
   bUseCustomPrimitiveData=true
   desc="For now Scalar, this is possible alternative to Texture Swatch. Uses single ramp texture shared between all 4 layers and can have indepent row selected per layer. Enabled globally via function. Will disable use of color swatch if enabled. Diffuse tint still can be used if needed via separate mask "
   group=Y_CPD
   parameterName=CPD_PackHSV_1
   primitiveDataIndex=30
   sliderMin=1
  }
  NamedRerouteDeclaration(None=MaterialFunctionCall(
    UV's=MaterialFunctionCall() {materialFunction=&MF_UVsSet_New}.UVs Layer 1
    Normals_NN=@n64
    SpecularMasks_NN=@n65
    Diffuse_NN=@n66
    Use SFX Emissive=StaticBoolParameter() {desc=Enables SFX, group=1_Layer SFX, parameterName=SFX Emissive, sortPriority=-8}
    SFX Opacity Mask=@n67
    SFX SSS Two Sided Foliange=StaticBoolParameter() {group=1_Layer SFX, parameterName=SFX_SSS Foliage, sortPriority=-2}
    SFX SSS=StaticBoolParameter() {group=1_Layer SFX, parameterName=SFX_SSS, sortPriority=-2}
    SFX Clearcoat=StaticBoolParameter() {group=1_Layer SFX, parameterName=SFX_Clearcoat, sortPriority=-5}
    Diffuse=NamedRerouteDeclaration(None=@n68) {name=L1D, nodeColor=<rgba 0 0.2969 1 1>}
    World Position Z=Constant() {r=1}
    SFX Texture Complete=NamedRerouteUsage()
    SFX Tint Channel=@n69.RGB
    SpecularMasks=NamedRerouteDeclaration(None=@n70) {name=L1S, nodeColor=<rgba 1 0.7734 0 1>}
    Normals=NamedRerouteDeclaration(None=@n71) {name=L1N, nodeColor=<rgba 1 0.3984 0 1>}
    Normals Power=ScalarParameter() {defaultValue=1, group=1_Layer, parameterName=NormalsValue, sortPriority=1}
    "SFX  SSS Channel"=VectorParameter() {group=1_Layer SFX, parameterName=SFX_SSS_Ch, sortPriority=-1}.RGB
    SSS Color Value=AppendVector(A=@n72, B=@n72)
    "Clearcoat (R) CC Roughness (G)"=@n73
    Clearcoat Mask Channal=VectorParameter() {group=1_Layer SFX, parameterName=SFX_Clearcoat_Chl, sortPriority=-4}.RGB
    SFX Opacilty Mask Channel=VectorParameter() {group=1_Layer SFX, parameterName=SFX_OpacityMask_Ch_1, sortPriority=2}.RGB
    Opacilty Mask Cancel=@n74
    "SFX Emissive  Channel"=VectorParameter() {group=1_Layer SFX, parameterName=SFX_Emissive_Ch, sortPriority=-7}.RGB
    Emissive Tint=AppendVector(A=@n75, B=@n75)
    Data Ramp RGB=MaterialFunctionCall!(In=@n76) {materialFunction=/Game/Global/Base4Layers_2022/MaterialFunctions/MF_ColorTint_Unpack}.Result
    Output Diffuse Only=StaticBoolParameter() {group=1_Layer, parameterName=Layer 1 Diffuse Only}
   ) {materialFunction=&MF_CombineLayers}.Result) {name=L1, nodeColor=<rgba 0 0.5078 1 1>}
  n77 = StaticBoolParameter() {
   desc=Enable to use UV2 for mapping. You may need to switch masks for ohter layers To UV3
   group=2_layer UV
   parameterName=Use UV2 for Layer 2
  }
  n78 = MaterialFunctionCall(L2 UV2=@n77) {materialFunction=&MF_UVsSet_New}
  n79 = StaticBoolParameter() {
   desc="World Space Z only, you need appropriate mask to make it visible. Use Z or map your layers accordingly "
   group=2_layer UV
   parameterName=Layer 2 WSM_Z
  }
  n80 = TextureObjectParameter() {
   group=2_Layer_Nanite
   parameterName="Normals  Nanite 2"
   samplerType=SAMPLERTYPE_Normal
   sortPriority=0
   texture=&FlatNormal
  }
  n81 = TextureObjectParameter() {
   group=2_Layer_Nanite
   parameterName="SpecularMasks  Nanite 2"
   samplerType=SAMPLERTYPE_Masks
   sortPriority=0
   texture=&T_Fortnite_Default_S
  }
  n82 = TextureObjectParameter() {group=2_Layer_Nanite, parameterName=Diffuse Nanite 2, sortPriority=0, texture=&T_MidgreysRGB}
  n83 = ScalarParameter() {
   bUseCustomPrimitiveData=true
   desc="For now Scalar, this is possible alternative to Texture Swatch. Uses single ramp texture shared between all 4 layers and can have indepent row selected per layer. Enabled globally via function. Will disable use of color swatch if enabled. Diffuse tint still can be used if needed via separate mask "
   group=Y_CPD
   parameterName=CPD_PackedHSV_2
   primitiveDataIndex=31
   sliderMin=1
  }
  n84 = StaticBool()
  n85 = StaticBoolParameter() {group=2_Layer_Blending, parameterName=Blend Overlay Layer 2}
  n86 = StaticSwitch(True=StaticBoolParameter() {group=2_Layer_Blending, parameterName=L2 Simple Swatch, sortPriority=-15}, False=@n84, Value=@n85)
  MaterialExpressionReroute_11 = Reroute(None=@n86)
  n87 = StaticSwitch(True=StaticBoolParameter() {group=2_Layer_Blending, parameterName=L2 Normals Only, sortPriority=-15}, False=@n84, Value=@n85)
  MaterialExpressionReroute_12 = Reroute(None=@n87)
  n88 = VectorParameter() {
   group=2_Layer SFX
   parameterName="SFX_SSS Tint (Expand)_2"
   primitiveDataIndex=2
   sortPriority=0
   channelNames {a=SSS Depth 0-1, b="", g="", r=""}
   defaultValue=<rgba 0 0 0 0.5>
  }
  n89 = VectorParameter() {
   desc="Use R and G channels of this input to controll ClearCoat amount and Rounghness. It uses SFX texture to mask area affected. "
   group=2_Layer SFX
   parameterName="SFX_ClearCoat(Expand)_2"
   sortPriority=-3
   channelNames {a=Not Used, b=Not Used, g=Clearcoat Roughness, r=Clearcoat Value}
   defaultValue=<rgba 1 0 0 0>
  }
  n90 = ScalarParameter() {desc=If you need to cancel shared Mask texture on specific layer, group=2_Layer SFX, parameterName=SFX_OpacityMaskCancel_2, sortPriority=3}
  n91 = VectorParameter() {
   desc="Emissive color, by default mult by Base color. You can control that blending via Alpha channel of this input. Uses SFX texture as the Mask"
   group=2_Layer SFX
   parameterName="SFX_Emissive_Tint(Expand)_2"
   sortPriority=-6
   channelNames {a=Base color to Emissive, b="", g="", r=""}
   defaultValue=<rgba 0 0 0 1>
  }
  n92 = TextureObjectParameter() {group=2_Layer, parameterName=Diffuse_Texture_2, sortPriority=0, texture=&T_MidgreysRGB}
  n93 = VectorParameter() {
   desc=Use R or G or B to select your SFX Mask Channel
   group=0_SFX BP
   parameterName=SFX_Tint_Ch_2
   sortPriority=4
   defaultValue=<rgba 1 0 0 0>
  }
  n94 = TextureObjectParameter() {
   group=2_Layer
   parameterName=SpecularMasks_2
   samplerType=SAMPLERTYPE_Masks
   sortPriority=0
   texture=&T_Fortnite_Default_S
  }
  n95 = TextureObjectParameter() {group=2_Layer, parameterName=Normals_Texture_2, samplerType=SAMPLERTYPE_Normal, sortPriority=0, texture=&FlatNormal}
  NamedRerouteDeclaration(None=MaterialFunctionCall(
    WSM Scale=@n78.WSM 2
    Use WS Mapping=StaticBoolParameter() {desc=Enable to use World Space Mapping for layer, group=2_layer UV, parameterName=Layer 2 WS Mapping}
    Use Z=@n79
    Use XYZ=StaticBoolParameter() {desc="World Space XYZ ( Expensive)", group=2_layer UV, parameterName=Layer 2 WSM_XYZ}
    UV2 True=@n77
    UV's=@n78.WSM 2
    Normals_NN=@n80
    SpecularMasks_NN=@n81
    Diffuse_NN=@n82
    Use SFX Emissive=StaticBoolParameter() {desc=Enables SFX, group=2_Layer SFX, parameterName=SFX Emissive 2, sortPriority=-8}
    SFX Opacity Mask=StaticBoolParameter() {desc=Enables SFX, group=2_Layer SFX, parameterName=SFX OpacityMask 2, sortPriority=1}
    SFX SSS Two Sided Foliange=StaticBoolParameter() {group=2_Layer SFX, parameterName=SFX_SSS Foliage 2, sortPriority=-2}
    SFX SSS=StaticBoolParameter() {group=2_Layer SFX, parameterName=SFX_SSS 2, sortPriority=-2}
    SFX Clearcoat=StaticBoolParameter() {group=2_Layer SFX, parameterName="SFX_Clearcoat  2", sortPriority=-5}
    Diffuse=NamedRerouteDeclaration(None=@n92) {name=D2, nodeColor=<rgba 1 0.3984 0 1>}
    World Position Z=ScalarParameter() {group=2_layer UV, parameterName=L2 WP Offset Z, sortPriority=1}
    SFX Texture Complete=NamedRerouteUsage()
    SFX Tint Channel=@n93.RGB
    SpecularMasks=NamedRerouteDeclaration(None=@n94) {name=S2, nodeColor=<rgba 1 0 0.3281 1>}
    Normals=NamedRerouteDeclaration(None=@n95) {name=N2, nodeColor=<rgba 0 0.5547 1 1>}
    Normals Power=ScalarParameter() {defaultValue=1, group=2_Layer, parameterName=NormalsValue_2, sortPriority=1}
    "SFX  SSS Channel"=VectorParameter() {group=2_Layer SFX, parameterName=SFX_SSS_Ch_2, sortPriority=-1}.RGB
    SSS Color Value=AppendVector(A=@n88, B=@n88)
    "Clearcoat (R) CC Roughness (G)"=@n89
    Clearcoat Mask Channal=VectorParameter() {group=2_Layer SFX, parameterName=SFX_Clearcoat_Ch_2, sortPriority=-4}.RGB
    SFX Opacilty Mask Channel=VectorParameter() {group=2_Layer SFX, parameterName=SFX_OpacityMask_Ch_2, sortPriority=2}.RGB
    Opacilty Mask Cancel=@n90
    "SFX Emissive  Channel"=VectorParameter() {group=2_Layer SFX, parameterName=SFX_Emissive_Ch_2, sortPriority=-7}.RGB
    Emissive Tint=AppendVector(A=@n91, B=@n91)
    Data Ramp RGB=MaterialFunctionCall(In=@n83) {materialFunction=&MF_ColorTint_Unpack}.Result
    Output Diffuse Only=@MaterialExpressionReroute_11.Output
    Output Normals Only=@MaterialExpressionReroute_12.Output2
   ) {materialFunction=&MF_CombineLayers}.Result) {name=L2, nodeColor=<rgba 1 0 0.5859 1>}
  NamedRerouteDeclaration(None=@n18) {name=SFX Texture Global, nodeColor=<rgba 0 1 0.5313 1>}
  n96 = TextureObjectParameter() {
   desc="SFX RGB - Graffiti, emissive RGB "
   group=0_SFX Global
   parameterName=SFX_RGB
   sortPriority=1
   texture=/Game/Global/Base4Layers_2022/Textures/T-GraffitiSample_D
  }
  NamedRerouteDeclaration(None=@n96) {name=SFX Texture RGB, nodeColor=<rgba 0 0.5312 1 1>}
  PinBase!(Output=@n11) {reroutePins [Output {expression=&self:MaterialExpressionReroute_9}]}
  PinBase!() {reroutePins [Input {expression=&self:MaterialExpressionReroute_3}]}
  PinBase() {reroutePins [InputPin {expression=&self:MaterialExpressionReroute_20}]}
  PinBase!(Output=@n15) {reroutePins [Output {expression=&self:MaterialExpressionReroute_21}]}
  PinBase!()
  PinBase!(Output=@n86, Output2=@n87) {reroutePins [
    Output {expression=&self:MaterialExpressionReroute_11}
    Output2 {expression=&self:MaterialExpressionReroute_12}
   ]}
  PinBase!()
  PinBase!(Output2=@n49, Output3=@n56) {reroutePins [
    Output2 {expression=&self:MaterialExpressionReroute_14}
    Output3 {expression=&self:MaterialExpressionReroute_15}
   ]}
  PinBase!()
  PinBase!(Output=@n26, Output2=@n33) {reroutePins [
    Output {expression=&self:MaterialExpressionReroute_16}
    Output2 {expression=&self:MaterialExpressionReroute_17}
   ]}
 ]
}
