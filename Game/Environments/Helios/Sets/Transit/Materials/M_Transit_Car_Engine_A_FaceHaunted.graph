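# Translucent, unlit, two-sided material that turns the haunted transit train's engine car into a ghostly smoke shape with a face.
# Channels of one texture supply face, shapes and two polar-mapped scrolling fog layers, with bump-offset parallax, blended into dark base and emissive colors.
# Opacity combines smoke, gradients, depth fade and edge fading from vertex color and view-dependent fresnel.
# Vertices displace along normals by gradient, attenuated with distance.
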Material /Game/Environments/Helios/Sets/Transit/Materials/M_Transit_Car_Engine_A_FaceHaunted
graph {
 properties {
  bAutomaticallySetUsageInEditor=false
  blendMode=BLEND_Translucent
  previewMesh=/Game/Environments/Helios/Sets/Transit/Meshes/TrainCars/SM_Transit_Car_Front_A_Haunted
  shadingModel=MSM_Unlit
  twoSided=true
  usedShadingModels=Unlit
 }
 nodes [
  n1 = Divide(A=NamedRerouteUsage()) {constB=15000}
  n2 = Multiply(A=OneMinus(None=NamedRerouteUsage()), B=ScalarParameter() {defaultValue=0.5, group=Train Haunted, parameterName=WPO Gradient})
  n3 = NamedRerouteUsage()
  n4 = ScalarParameter() {defaultValue=90, group=Train Haunted, parameterName=WPO}
  n5 = Multiply(A=Multiply(A=Saturate(None=@n2), B=Multiply(A=@n3, B=@n4)), B=VertexNormalWS())
  n6 = Constant()
  n7 = NamedRerouteUsage()
  n8 = MaterialFunctionCall(In=Add(A=NamedRerouteUsage(), B=Constant() {r=0.4}), Contrast=Constant() {r=2}) {materialFunction=&CheapContrast}
  LinearInterpolate(A=LinearInterpolate(A=@n5, B=@n6, Alpha=Multiply(A=@n7, B=@n8.Result)), B=Constant(), Alpha=Saturate(None=Subtract(A=@n1) {constB=0.2})) {constB=0}
  Multiply(A=@n5, B=@n6) {constB=0.5}
  n9 = OneMinus(None=VertexColor().R)
  n10 = ScalarParameter() {defaultValue=0.6, desc=0.6, group=Train Haunted, parameterName=Fade_Fresnel Sides Min}
  n11 = ScalarParameter() {desc=0.27, group=Train Haunted, parameterName=Fade_Fresnel Sides Max}
  n12 = Abs(None=DotProduct(A=CameraVectorWS(), B=VertexNormalWS()))
  n13 = ScalarParameter() {defaultValue=0.4, desc=0.4, group=Train Haunted, parameterName=Fade_Fresnel Front Min}
  n14 = ScalarParameter() {desc=0, group=Train Haunted, parameterName=Fade_Fresnel front Max}
  n15 = DotProduct(
   A=MaterialFunctionCall() {materialFunction=&CameraDirectionVector}.Result
   B=Transform(None=Constant3Vector() {constant=<rgba -1 0 0 0>}) {transformSourceType=TRANSFORMSOURCE_Local}
  )
  n16 = SmoothStep(Value=@n15) {constMin=0.9}
  n17 = SmoothStep(Min=Constant() {desc=-0.7, r=-0.9}, Value=@n15) {constMax=-1, constMin=-0.9}
  n18 = LinearInterpolate(
   A=LinearInterpolate(
    A=SmoothStep(Min=@n10, Max=@n11, Value=@n12) {constMax=0, constMin=0.3}
    B=SmoothStep(Min=@n13, Max=@n14, Value=@n12) {constMax=0, constMin=0.3}
    Alpha=VertexInterpolator(VS=@n16) {desc=Front View}.PS
   ) {constA=1}
   B=SmoothStep(
    Min=ScalarParameter() {defaultValue=0.2, group=Train Haunted, parameterName=Fade_Fresnel Back Min}
    Max=ScalarParameter() {group=Train Haunted, parameterName=Fade_Fresnel Back Max}
    Value=@n12
   ) {constMax=0, constMin=0.3}
   Alpha=VertexInterpolator(VS=NamedRerouteDeclaration(None=@n17) {name=back view, nodeColor=<rgba 1 0 0.9609 1>}) {desc=back view}.PS
  )
  NamedRerouteDeclaration(None=Max(
    A=NamedRerouteDeclaration(None=@n9) {name=M_VertexEdges, nodeColor=<rgba 0 0.7656 1 1>}
    B=NamedRerouteDeclaration(None=@n18) {name=M_Fresnel, nodeColor=<rgba 0.6875 1 0 1>}
   )) {name=M_Fade Edges, nodeColor=<rgba 1 0 0.6328 1>}
  n19 = PixelDepth()
  NamedRerouteDeclaration(None=@n19.R) {name=IN_PixelDepht, nodeColor=<rgba 1 0 0.4453 1>}
  n20 = TextureCoordinate()
  n21 = ScalarParameter() {defaultValue=-0.205, desc=-0.205, group=Train Haunted, parameterName=Face_Pos_X}
  n22 = Add(A=@n21) {constB=-0.165}
  n23 = ScalarParameter() {defaultValue=-0.65, desc=-0.65, group=Train Haunted, parameterName=Face_Pos_Y}
  n24 = Add(A=Constant() {r=-0.06}, B=@n23)
  n25 = Multiply(A=Add(A=@n20, B=AppendVector(A=@n22, B=@n24))) {constB=3.9}
  n26 = Constant() {r=0.5}
  NamedRerouteDeclaration(None=OneMinus(None=Distance(A=@n25, B=@n26))) {name=M_GradientRadial, nodeColor=<rgba 1 0.9141 0 1>}
  NamedRerouteDeclaration(None=Multiply(A=Time(), B=ScalarParameter() {defaultValue=1, group=Train Haunted, parameterName=Time_Mult})) {name=IN_Time, nodeColor=<rgba 1 0.5859 0 1>}
  n27 = Constant() {r=-0.06}
  n28 = Constant()
  n29 = MaterialFunctionCall!(UV Coord=Add(A=TextureCoordinate(), B=Constant2Vector() {g=-0.6}), LenghtScale, RadialScale) {materialFunction=/Game/Global/MaterialFunctions/TechArt/MF_PolarCoordinates}
  n30 = Add(A=@n29.Result, B=Multiply(A=NamedRerouteUsage(), B=AppendVector(A=@n27, B=@n28)))
  n31 = Constant2Vector() {g=50, r=40}
  n32 = Add(A=TextureSample(UVs=Multiply(A=@n30, B=@n31) {constB=300}) {samplerType=SAMPLERTYPE_Grayscale, texture=&LowResBlurredNoise}.R) {constB=-0.5}
  n33 = Saturate(None=Divide(A=@n19.R, B=Constant() {r=1024}) {constB=1024})
  NamedRerouteDeclaration(None=Multiply(A=Multiply(A=@n32) {constB=-0.02}, B=OneMinus(None=@n33))) {name=UV_Noise, nodeColor=<rgba 0.0078 0 1 1>}
  n34 = NamedRerouteUsage()
  n35 = Multiply(A=@n34, B=@n34) {constB=2}
  n36 = Multiply(A=Multiply(A=@n35, B=@n35), B=Constant() {r=1.5})
  n37 = Max(
   A=NamedRerouteDeclaration(None=Saturate(None=@n36)) {name=M_GradientTail2, nodeColor=<rgba 0.9922 0 1 1>}
   B=Multiply(A=NamedRerouteUsage(), B=Constant() {r=0.3}) {constB=0.5}
  )
  n38 = NamedRerouteUsage()
  n39 = SmoothStep(
   Min=ScalarParameter() {defaultValue=-0.5, desc=-0.5, group=Train Haunted, parameterName=Opacity_Min}
   Max=ScalarParameter() {defaultValue=0.5, desc=0.5, group=Train Haunted, parameterName=Opacity_Max}
   Value=Subtract(A=NamedRerouteUsage(), B=Max(A=@n37, B=@n38))
  ) {constMax=0.05, constMin=-0.2}
  n40 = NamedRerouteUsage()
  n41 = Constant() {r=0.5}
  n42 = LinearInterpolate(A=Multiply(A=@n39, B=Constant() {r=0.9}) {constB=0.95}, B=@n39, Alpha=Max(A=Multiply(A=@n40, B=@n41), B=NamedRerouteUsage()))
  n43 = OneMinus(None=NamedRerouteUsage())
  n44 = Constant() {r=1}
  n45 = Constant() {r=36}
  Reroute(None=Multiply(A=Multiply(A=@n42, B=@n43), B=DepthFade(Opacity=@n44, FadeDistance=@n45)))
  n46 = VectorParameter() {group=Train Haunted, parameterName=Base_Color, defaultValue=<rgba 0.04 0.0533 0.08 1>}
  n47 = NamedRerouteUsage()
  n48 = AppendVector(
   A=ScalarParameter() {defaultValue=-0.2, group=Train Haunted, parameterName=Noise_Speed_X}
   B=ScalarParameter() {group=Train Haunted, parameterName=Noise_Speed_Y}
  )
  n49 = Multiply(
   A=Add(A=@n29.Result, B=Multiply(A=@n47, B=@n48))
   B=AppendVector(
    A=ScalarParameter() {defaultValue=2, group=Train Haunted, parameterName=Noise_Tile_X}
    B=ScalarParameter() {defaultValue=5, group=Train Haunted, parameterName=Noise_Tile_Y}
   )
  )
  n50 = NamedRerouteUsage()
  n51 = NamedRerouteDeclaration(None=TextureSample(UVs=Add(A=@n49, B=@n50)) {samplerType=SAMPLERTYPE_LinearColor, texture=/Game/Environments/Helios/Sets/Transit/Textures/T_FakeInteriorHaunted_Face}.B) {name=Fog1, nodeColor=<rgba 0 0.7891 1 1>}
  n52 = BumpOffset(Height=Add(A=Multiply(A=@n51) {constB=50}, B=Constant() {desc=-20, r=-20}) {constB=-20})
  n53 = NamedRerouteUsage()
  n54 = AppendVector(
   A=ScalarParameter() {defaultValue=-0.17, group=Train Haunted, parameterName=Noise2_Speed_X}
   B=ScalarParameter() {group=Train Haunted, parameterName=Noise2_Speed_Y}
  )
  n55 = Multiply(
   A=Add(A=@n29.Result, B=Multiply(A=@n53, B=@n54))
   B=AppendVector(
    A=ScalarParameter() {defaultValue=8, group=Train Haunted, parameterName=Nois2e_Tile_X}
    B=ScalarParameter() {defaultValue=20, group=Train Haunted, parameterName=Noise2_Tile_Y}
   )
  )
  n56 = OneMinus(None=TextureSample(UVs=Add(A=@n52, B=@n55)) {samplerType=SAMPLERTYPE_LinearColor, texture=&T_FakeInteriorHaunted_Face}.B)
  n57 = ScalarParameter() {defaultValue=0.3, group=Train Haunted, parameterName=Fog2_Mult, sortPriority=5}
  n58 = Constant() {r=0.13}
  n59 = Constant() {r=1.13}
  n60 = Sine(None=Divide(A=NamedRerouteUsage()) {constB=5})
  n61 = Multiply(A=Multiply(A=Constant() {r=-0.5}, B=Add(A=@n60)), B=TwoSidedSign())
  n62 = Multiply(
   A=Add(A=BumpOffset(Coordinate=@n20, Height=@n61), B=AppendVector(A=@n21, B=@n23))
   B=AppendVector(
    A=ScalarParameter() {defaultValue=1.7, desc=1.7, group=Train Haunted, parameterName=Face_Tile_X}
    B=ScalarParameter() {defaultValue=2.8, desc=-0.65, group=Train Haunted, parameterName=Face_Tile_Y}
   )
  ) {constB=3.9}
  n63 = Multiply(A=OneMinus(None=Step(X=@n62) {constX=0.5, constY=1}), B=Step(X=@n62) {constX=0.5})
  n64 = Add(A=@n62, B=Multiply(A=NamedRerouteUsage()) {constB=0.01})
  n65 = NamedRerouteUsage()
  n66 = Multiply(
   A=Multiply(A=ComponentMask(None=@n63) {r=true}, B=ComponentMask(None=@n63) {g=true})
   B=TextureSample(UVs=Add(A=@n64, B=@n65)) {samplerType=SAMPLERTYPE_LinearColor, texture=&T_FakeInteriorHaunted_Face}.R
  )
  n67 = Multiply(
   A=NamedRerouteDeclaration(None=SmoothStep(Min=@n58, Max=@n59, Value=@n66) {constMax=0.8, constMin=0.15}) {name=M_Face, nodeColor=<rgba 1 0 0.375 1>}
   B=ScalarParameter() {defaultValue=0.6, group=Train Haunted, parameterName=Face_Mult, sortPriority=5}
  ) {constB=2}
  n68 = TextureSample(UVs=Add(A=TextureCoordinate(), B=@n65)) {samplerType=SAMPLERTYPE_LinearColor, texture=&T_FakeInteriorHaunted_Face}
  n69 = Add(
   A=Multiply(A=Add(A=@n68.G) {constB=-0.2}, B=ScalarParameter() {defaultValue=0.63, group=Train Haunted, parameterName=Shapes_Mult, sortPriority=5}) {constB=0.5}
   B=Multiply(A=@n51, B=ScalarParameter() {defaultValue=1, group=Train Haunted, parameterName=Fog_Mult, sortPriority=5})
  )
  n70 = Add(
   A=Subtract(A=Add(A=@n67, B=@n69), B=Multiply(A=@n56, B=@n57) {constB=0.5})
   B=ScalarParameter() {defaultValue=0.6, group=Train Haunted, parameterName=Mask Value Bias}
  ) {constB=0}
  n71 = OneMinus(None=ComponentMask(None=TextureCoordinate()) {g=true})
  n72 = Max(
   A=NamedRerouteUsage()
   B=SmoothStep(
    Min=ScalarParameter() {defaultValue=-0.5, group=Train Haunted, parameterName=Linear Gradient_Min, sortPriority=5}
    Max=ScalarParameter() {defaultValue=3, group=Train Haunted, parameterName=Linear Gradient_Max, sortPriority=5}
    Value=NamedRerouteDeclaration(None=@n71) {name=M_GradientTail, nodeColor=<rgba 0.8281 1 0 1>}
   )
  )
  n73 = Add(A=NamedRerouteUsage()) {constB=0.5}
  n74 = ScalarParameter() {defaultValue=0.3, group=Train Haunted, parameterName=FaceGradient_Mult, sortPriority=5}
  n75 = NamedRerouteDeclaration(None=Saturate(None=Multiply(A=@n73, B=@n74))) {name=FaceGradient, nodeColor=<rgba 0.9922 1 0 1>}
  n76 = Multiply(
   A=Subtract(A=NamedRerouteDeclaration(None=@n70) {name=SmokeValues, nodeColor=<rgba 0.3125 1 0 1>}, B=Max(A=@n72, B=@n75))
   B=OneMinus(None=@n72)
  )
  n77 = LinearInterpolate(
   A=@n46.RGB
   B=Multiply(A=@n46.RGB) {constB=0.2}
   Alpha=SmoothStep(Min=Constant() {r=-1}, Max=Constant() {r=0.6}, Value=@n76) {constMax=0.7, constMin=0.1}
  )
  n78 = Multiply(A=Multiply(A=@n76, B=@n76), B=ScalarParameter() {defaultValue=0.5, group=Train Haunted, parameterName=Emissive_Smoke Intensity})
  n79 = NamedRerouteUsage()
  n80 = LinearInterpolate(
   A=VectorParameter() {group=Train Haunted, parameterName=Emissive_Color1, defaultValue=<rgba 0.012 0.066 0.12 1>}.RGB
   B=VectorParameter() {group=Train Haunted, parameterName=Emissive_Color2, defaultValue=<rgba 0 1.8 1.11 1>}.RGB
   Alpha=Saturate(None=LinearInterpolate(A=@n78, B=@n76, Alpha=@n79) {constAlpha=0.3})
  )
  n81 = Constant()
  n82 = Constant() {r=0.6}
  n83 = Add(A=NamedRerouteUsage(), B=Constant() {r=-0.5})
  n84 = Constant() {r=2}
  n85 = Saturate(None=Subtract(A=SmoothStep(Min=@n81, Max=@n82, Value=@n76) {constMax=0.7, constMin=0.1}, B=Multiply(A=@n83, B=@n84)))
  Reroute(None=Multiply(A=LinearInterpolate(A=@n77, B=@n80, Alpha=@n85), B=EyeAdaptationInverse().EyeAdaptationInverse))
 ]
}
